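# Control Rig for the Kitty Warrior player body that applies corrective bone offsets to the pants, legs and neck collar.
# When not DBNO, FixCrouchCrashing uses a spherical pose reader on the left calf to weight offsets on both legs' bones.
# FixNeckCollar always runs, shifting and scaling the collar by the larger of two swing-twist neck rotation weights.
# Its NeckDelta debug print is left enabled.
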ControlRigBlueprint /BRCosmetics/Characters/Player/Male/Medium/Bodies/M_MED_Kitty_Warrior/Meshes/M_MED_Kitty_Warrior_CtrlRig
graphs [
 New Function_ContainedGraph {nodes [
   RigVMFunctionEntryNode()
   > n1 = &RigUnit_SphericalPoseReader(
    ActiveRegionScaleFactors=(PositiveWidth=0.7,NegativeWidth=0.7,PositiveHeight=0.7,NegativeHeight=0.7)
    ActiveRegionSize=0.1
    Debug=(bDrawDebug=false,bDraw2D=false,bDrawLocalAxes=False,DebugScale=25,DebugSegments=20,DebugThickness=0.25)
    DriverAxis=(X=-1,Y=0,Z=0)
    DriverItem=(Type=Bone,Name="calf_l")
    FalloffRegionScaleFactors=(PositiveWidth=0.6,NegativeWidth=0.6,PositiveHeight=0.6,NegativeHeight=0.6)
    FalloffSize=0.2
    FlipHeightScaling=False
    FlipWidthScaling=False
    OptionalParentItem=(Type=Bone,Name="None")
    RotationOffset=(X=90,Y=0,Z=-140)
   )
   > RigVMVariableNode(Value=@n1.OutputParam) {variable=L_KneeBendValue}
   > &RigUnit_OffsetTransformForItem(
    Item=(Type=Bone,Name="pants_l_back_01_reverse")
    OffsetTransform=(Rotation=(X=0,Y=0,Z=0.342,W=0.9397),Translation=(X=0,Y=0,Z=-20),Scale3D=(X=1,Y=1,Z=1))
    bPropagateToChildren=True
    Weight=RigVMVariableNode() {variable=L_KneeBendValue}.Value
   )
   > &RigUnit_OffsetTransformForItem(
    Item=(Type=Bone,Name="pants_l_back_02_reverse")
    OffsetTransform=(Rotation=(X=0,Y=0,Z=-0.1736,W=0.9848),Translation=(X=0,Y=-20,Z=20),Scale3D=(X=1,Y=1,Z=1))
    bPropagateToChildren=True
    Weight=RigVMVariableNode() {variable=L_KneeBendValue}.Value
   )
   > &RigUnit_OffsetTransformForItem(
    Item=(Type=Bone,Name="dyn_pants_l_back_03")
    OffsetTransform=(Rotation=(X=0,Y=0,Z=-0.1305,W=0.9914),Translation=(X=-5,Y=0,Z=0),Scale3D=(X=1,Y=1,Z=1))
    bPropagateToChildren=True
    Weight=RigVMVariableNode() {variable=L_KneeBendValue}.Value
   )
   > &RigUnit_OffsetTransformForItem(
    Item=(Type=Bone,Name="pants_r_back_01_reverse")
    OffsetTransform=(Rotation=(X=0,Y=0,Z=0.1736,W=0.9848),Translation=(X=0,Y=7,Z=15),Scale3D=(X=1,Y=1,Z=1))
    bPropagateToChildren=True
    Weight=RigVMVariableNode() {variable=L_KneeBendValue}.Value
   )
   > &RigUnit_OffsetTransformForItem(
    Item=(Type=Bone,Name="pants_r_back_02_reverse")
    OffsetTransform=(Rotation=(X=0,Y=0,Z=0.5,W=0.866),Translation=(X=30,Y=55,Z=20),Scale3D=(X=1,Y=1,Z=1))
    bPropagateToChildren=True
    Weight=RigVMVariableNode() {variable=L_KneeBendValue}.Value
   )
   > &RigUnit_OffsetTransformForItem(
    Item=(Type=Bone,Name="dyn_pants_r_back_03")
    OffsetTransform=(Rotation=(X=0,Y=0,Z=-0.2164,W=0.9763),Translation=(X=0,Y=-2,Z=-6),Scale3D=(X=1,Y=1,Z=1))
    bPropagateToChildren=True
    Weight=RigVMVariableNode() {variable=L_KneeBendValue}.Value
   )
   > &RigUnit_OffsetTransformForItem(
    Item=(Type=Bone,Name="thigh_twist_01_r")
    OffsetTransform=(Rotation=(X=0,Y=0,Z=0,W=1),Translation=(X=0,Y=-6,Z=0),Scale3D=(X=1,Y=0.9,Z=1))
    bPropagateToChildren=True
    Weight=RigVMVariableNode() {variable=L_KneeBendValue}.Value
   )
   > &RigUnit_OffsetTransformForItem(
    Item=(Type=Bone,Name="thigh_twist_01_l")
    OffsetTransform=(Rotation=(X=0,Y=0,Z=0,W=1),Translation=(X=0,Y=1,Z=0),Scale3D=(X=1,Y=0.9,Z=1))
    bPropagateToChildren=True
    Weight=RigVMVariableNode() {variable=L_KneeBendValue}.Value
   )
   > &RigUnit_OffsetTransformForItem(
    Item=(Type=Bone,Name="calf_twist_02_r")
    OffsetTransform=(Rotation=(X=0,Y=0,Z=0.0872,W=0.9962),Translation=(X=0,Y=-2,Z=0),Scale3D=(X=1,Y=1,Z=1))
    bPropagateToChildren=True
    Weight=RigVMVariableNode() {variable=L_KneeBendValue}.Value
   )
   > &RigUnit_OffsetTransformForItem(
    Item=(Type=Bone,Name="dyn_pants_r_front_01")
    OffsetTransform=(Rotation=(X=0,Y=0,Z=0.0436,W=0.999),Translation=(X=0,Y=-2,Z=0),Scale3D=(X=1.02,Y=1,Z=1))
    bPropagateToChildren=True
    Weight=RigVMVariableNode() {variable=L_KneeBendValue}.Value
   )
   > RigVMFunctionReturnNode()
  ]}
 New Function_ContainedGraph {nodes [
   RigVMFunctionEntryNode()
   > DeltaRotation = RigVMCollapseNode(
    Rotation Order=ZYX
    child=neck_02
    debugRotation=false
    max=-15
    min=0
    parent=spine_05
    parentType=Bone
    premult=1
    primAxis=1
    twistAxis=(X=1,Y=0,Z=0)
   )
   > DeltaRotation_1 = RigVMCollapseNode(
    Rotation Order=ZYX
    child=neck_02
    debugRotation=false
    max=-15
    min=0
    parent=neck_01
    parentType=Bone
    premult=1
    primAxis=2
    twistAxis=(X=1,Y=0,Z=0)
   )
   Maximum = &RigVMFunction_MathDoubleMax(A=@DeltaRotation_1.outputWeight, B=@DeltaRotation.outputWeight)
   > &RigUnit_OffsetTransformForItem(
    Item=(Type=Bone,Name="neck_collar")
    OffsetTransform=(Rotation=(X=0,Y=0,Z=0,W=1),Translation=(X=0,Y=-5,Z=0),Scale3D=(X=1,Y=1,Z=1.2))
    bPropagateToChildren=True
    Weight=@Maximum.Result
   )
   > RigVMFunctionReturnNode()
   RigVMDispatchNode(Enabled=True, ExecuteContext=(), Prefix=NeckDelta, ScreenColor=(R=1,G=1,B=1,A=1), ScreenDuration=0.05, Value=@Maximum.Result) {template="DISPATCH_RigVMDispatch_Print(in Prefix,in Value,in Enabled,in ScreenDuration,in ScreenColor)"}
  ]}
 New Function_ContainedGraph {nodes [
   n1 = RigVMFunctionEntryNode()
   RigVMFunction_MathBoolToInteger = &RigVMFunction_MathBoolToInteger(Value=@n1.onlyTwist)
   RigVMFunction_MathQuaternionSwingTwist = &RigVMFunction_MathQuaternionSwingTwist(
    TwistAxis=@n1.twistAxis
    Input=&RigUnit_GetRelativeTransformForItem(
     Child=(Type=Bone,Name="None")
     Parent=(Type=Bone,Name="None")
     bChildInitial=False
     bParentInitial=False
     Child.Name=@n1.child
     Parent.Name=@n1.parent
     Parent.Type=@n1.parentType
    ).RelativeTransform.Rotation
   )
   RigVMFunction_MathQuaternionToEuler = &RigVMFunction_MathQuaternionToEuler(
    Value=RigVMDispatchNode(
     Values=((),())
     Index=@RigVMFunction_MathBoolToInteger.Result
     Values.0=@RigVMFunction_MathQuaternionSwingTwist.Swing
     Values.1=@RigVMFunction_MathQuaternionSwingTwist.Twist
    ) {template="DISPATCH_RigVMDispatch_SelectInt32(in Index,in Values,out Result)"}.Result
    RotationOrder=@n1.Rotation Order
   )
   > DISPATCH_RigVMDispatch_Print = RigVMDispatchNode(ScreenColor=(R=1,G=1,B=1,A=1), ScreenDuration=1.5, Enabled=@n1.debugRotation, Value=@RigVMFunction_MathQuaternionToEuler.Result) {template="DISPATCH_RigVMDispatch_Print(in Prefix,in Value,in Enabled,in ScreenDuration,in ScreenColor)"}
   Clamp = &RigVMFunction_MathIntClamp(Value=@n1.primAxis)
   DISPATCH_RigVMDispatch_SelectInt32 = RigVMDispatchNode(
    Values=(0,0,0)
    Index=@Clamp.Result
    Values.0=@RigVMFunction_MathQuaternionToEuler.Result.X
    Values.1=@RigVMFunction_MathQuaternionToEuler.Result.Y
    Values.2=@RigVMFunction_MathQuaternionToEuler.Result.Z
   ) {template="DISPATCH_RigVMDispatch_SelectInt32(in Index,in Values,out Result)"}
   Remap = &RigVMFunction_MathDoubleRemap(
    TargetMaximum=1
    TargetMinimum=0
    bClamp=False
    Value=@DISPATCH_RigVMDispatch_SelectInt32.Result
    SourceMaximum=@n1.max
    SourceMinimum=@n1.min
   )
   Multiply = &RigVMFunction_MathDoubleMul(A=&RigVMFunction_MathDoubleClamp(Value=@Remap.Result).Result, B=@n1.premult)
   > RigVMFunctionReturnNode(outputWeight=@Multiply.Result)
  ]}
 New Function_ContainedGraph {nodes [
   n1 = RigVMFunctionEntryNode()
   RigVMFunction_MathBoolToInteger = &RigVMFunction_MathBoolToInteger(Value=@n1.onlyTwist)
   RigVMFunction_MathQuaternionSwingTwist = &RigVMFunction_MathQuaternionSwingTwist(
    TwistAxis=@n1.twistAxis
    Input=&RigUnit_GetRelativeTransformForItem(
     Child=(Type=Bone,Name="None")
     Parent=(Type=Bone,Name="None")
     bChildInitial=False
     bParentInitial=False
     Child.Name=@n1.child
     Parent.Name=@n1.parent
     Parent.Type=@n1.parentType
    ).RelativeTransform.Rotation
   )
   RigVMFunction_MathQuaternionToEuler = &RigVMFunction_MathQuaternionToEuler(
    Value=RigVMDispatchNode(
     Values=((),())
     Index=@RigVMFunction_MathBoolToInteger.Result
     Values.0=@RigVMFunction_MathQuaternionSwingTwist.Swing
     Values.1=@RigVMFunction_MathQuaternionSwingTwist.Twist
    ) {template="DISPATCH_RigVMDispatch_SelectInt32(in Index,in Values,out Result)"}.Result
    RotationOrder=@n1.Rotation Order
   )
   > DISPATCH_RigVMDispatch_Print = RigVMDispatchNode(ScreenColor=(R=1,G=1,B=1,A=1), ScreenDuration=1.5, Enabled=@n1.debugRotation, Value=@RigVMFunction_MathQuaternionToEuler.Result) {template="DISPATCH_RigVMDispatch_Print(in Prefix,in Value,in Enabled,in ScreenDuration,in ScreenColor)"}
   Clamp = &RigVMFunction_MathIntClamp(Value=@n1.primAxis)
   DISPATCH_RigVMDispatch_SelectInt32 = RigVMDispatchNode(
    Values=(0,0,0)
    Index=@Clamp.Result
    Values.0=@RigVMFunction_MathQuaternionToEuler.Result.X
    Values.1=@RigVMFunction_MathQuaternionToEuler.Result.Y
    Values.2=@RigVMFunction_MathQuaternionToEuler.Result.Z
   ) {template="DISPATCH_RigVMDispatch_SelectInt32(in Index,in Values,out Result)"}
   Remap = &RigVMFunction_MathDoubleRemap(
    TargetMaximum=1
    TargetMinimum=0
    bClamp=False
    Value=@DISPATCH_RigVMDispatch_SelectInt32.Result
    SourceMaximum=@n1.max
    SourceMinimum=@n1.min
   )
   Multiply = &RigVMFunction_MathDoubleMul(A=&RigVMFunction_MathDoubleClamp(Value=@Remap.Result).Result, B=@n1.premult)
   > RigVMFunctionReturnNode(outputWeight=@Multiply.Result)
  ]}
 RigVMFunctionLibrary {nodes [
   FixCrouchCrashing = RigVMCollapseNode(ExecuteContext=())
   FixNeckCollar = RigVMCollapseNode(ExecuteContext=())
  ]}
 RigVMModel {nodes [
   RigUnit_BeginExecution = &RigUnit_BeginExecution()
   RigVMFunction_ControlFlowBranch = &RigVMFunction_ControlFlowBranch(ExecuteContext=@RigUnit_BeginExecution.ExecutePin, Condition=RigVMVariableNode() {variable=Is DBNO}.Value)
   FixNeckCollar = RigVMFunctionReferenceNode(ExecuteContext=@RigVMFunction_ControlFlowBranch.Completed)
   FixCrouchCrashing = RigVMFunctionReferenceNode(ExecuteContext=@RigVMFunction_ControlFlowBranch.False)
  ]}
]
hierarchy [
 root {loc=<0 0 0>, rot=<0 0 0 1>}
 attach {parent=root, loc=<0 0 0>, rot=<0 0 0 1>}
 pelvis {parent=root, loc=<0 -0.7071 93.1191>, rot=<0.0127 0.707 -0.0127 -0.707>}
 spine_01 {parent=pelvis, loc=<6.7494 -0.6388 0>, rot=<0 0 0.034 -0.9994>}
 spine_02 {parent=spine_01, loc=<7.2518 -0.035 0>, rot=<0 0 0.0238 -0.9997>}
 spine_03 {parent=spine_02, loc=<6.0104 0 0>, rot=<0 0 0.0149 -0.9999>}
 spine_04 {parent=spine_03, loc=<8.1024 0 0>, rot=<0 0 -0.0132 -0.9999>}
 spine_05 {parent=spine_04, loc=<10.0146 0 0>, rot=<0 0 -0.0015 -1>}
 clavicle_l {parent=spine_05, loc=<4.3575 -0.1466 -2.3223>, rot=<0.0465 -0.7289 -0.009 -0.6829>}
 upperarm_l {parent=clavicle_l, loc=<15.5111 0 0>, rot=<0.0226 -0.3234 -0.0088 -0.9459>}
 lowerarm_l {parent=upperarm_l, loc=<26.6514 0 0>, rot=<0.0391 -0.0099 -0.2453 -0.9686>}
 hand_l {parent=lowerarm_l, loc=<22.3491 0 0>, rot=<-0.7081 -0.0507 0.0975 -0.6975>}
 index_metacarpal_l {parent=hand_l, loc=<1.6535 0.0183 -2.96>, rot=<0.0158 -0.0124 -0.0173 -0.9996>}
 index_01_l {parent=index_metacarpal_l, loc=<8.3008 0.0357 -0.0524>, rot=<-0.0044 0.0485 0.0924 -0.9945>}
 index_02_l {parent=index_01_l, loc=<3.3587 -0.3004 0.0452>, rot=<-0.002 0.0027 0.3225 -0.9466>}
 index_03_l {parent=index_02_l, loc=<2.5578 -0.123 -0.0252>, rot=<-0.002 0.0067 0.2558 -0.9667>}
 middle_metacarpal_l {parent=hand_l, loc=<2.1856 -0.1725 -0.5984>, rot=<0.006 0.0168 -0.0329 -0.9993>}
 middle_01_l {parent=middle_metacarpal_l, loc=<7.3275 -0.2447 0.0129>, rot=<-0.0208 0.0042 0.0858 -0.9961>}
 middle_02_l {parent=middle_01_l, loc=<4.0123 0.1401 0.1175>, rot=<-0.0218 0.0258 0.3512 -0.9357>}
 middle_03_l {parent=middle_02_l, loc=<2.6077 0.0534 0.1373>, rot=<0.0082 0.0188 0.1708 -0.9851>}
 pinky_metacarpal_l {parent=hand_l, loc=<1.7616 -0.9753 3.3814>, rot=<-0.0397 0.0814 -0.069 -0.9935>}
 pinky_01_l {parent=pinky_metacarpal_l, loc=<7.1709 -0.4831 -0.1094>, rot=<-0.0654 -0.0035 0.2242 -0.9723>}
 pinky_02_l {parent=pinky_01_l, loc=<2.8274 0.0893 -0.1043>, rot=<0.0037 0.0011 0.2854 -0.9584>}
 pinky_03_l {parent=pinky_02_l, loc=<2.2046 0.0307 -0.0078>, rot=<-0.0002 0.0098 0.1445 -0.9895>}
 ring_metacarpal_l {parent=hand_l, loc=<2.2074 -0.3625 1.3638>, rot=<-0.0284 0.0722 -0.0458 -0.9959>}
 ring_01_l {parent=ring_metacarpal_l, loc=<7.0295 -0.6197 -0.1446>, rot=<-0.0489 -0.0252 0.1892 -0.9804>}
 ring_02_l {parent=ring_01_l, loc=<3.6174 0.4647 -0.1027>, rot=<0 0 0.2806 -0.9598>}
 ring_03_l {parent=ring_02_l, loc=<2.4152 0.114 0.0199>, rot=<0 0 0.1189 -0.9929>}
 thumb_01_l {parent=hand_l, loc=<1.3118 -0.3191 -2.459>, rot=<0.6546 -0.4042 0.0664 -0.6354>}
 thumb_02_l {parent=thumb_01_l, loc=<5.9336 -0.0448 0.0401>, rot=<-0.0037 0.0213 0.298 -0.9543>}
 thumb_03_l {parent=thumb_02_l, loc=<2.2223 0.3682 -0.0907>, rot=<0.0118 -0.013 0.0066 -0.9998>}
 weapon_l {parent=hand_l, loc=<-0.1974 -3.4111 0.3811>, rot=<0 0 0 1>}
 lowerarm_twist_01_l {parent=lowerarm_l, loc=<17.8549 0 0>, rot=<0 0 0 1>}
 lowerarm_twist_02_l {parent=lowerarm_l, loc=<9.6199 0 0>, rot=<0 0 0 1>}
 dyn_ribbon_knot_l_MSTR {parent=lowerarm_l, loc=<0.9428 1.5563 4.526>, rot=<0.0951 0.5724 0.2096 -0.787>}
 dyn_ribbon_l_loop_02 {parent=dyn_ribbon_knot_l_MSTR, loc=<1.0609 -0.1812 0.1275>, rot=<0.8956 -0.3591 0.2438 -0.0978>}
 dyn_ribbon_l_string_b_01 {parent=dyn_ribbon_knot_l_MSTR, loc=<1.5392 -0.0901 -0.9773>, rot=<-0.0266 -0.4339 -0.0551 -0.8989>}
 dyn_ribbon_l_string_b_02 {parent=dyn_ribbon_l_string_b_01, loc=<7.0438 0 0>, rot=<0.0115 0.0932 0.0002 -0.9956>}
 dyn_ribbon_l_string_a_01 {parent=dyn_ribbon_knot_l_MSTR, loc=<0.5855 1.3446 -0.4098>, rot=<-0.1997 -0.5529 -0.2748 -0.7609>}
 dyn_ribbon_l_string_a_02 {parent=dyn_ribbon_l_string_a_01, loc=<6.8586 0 0>, rot=<0.1342 0.1911 0.0788 -0.9691>}
 dyn_ribbon_l_loop_01 {parent=dyn_ribbon_knot_l_MSTR, loc=<0.4032 1.2352 0.4781>, rot=<0.6578 0.2353 0.6736 0.241>}
 upperarm_twist_01_l {parent=upperarm_l, loc=<0.0041 0 0>, rot=<0 0 0 1>}
 upperarm_twist_02_l {parent=upperarm_l, loc=<13.7941 0 0>, rot=<0 0 0 1>}
 dyn_shoulder_l {parent=clavicle_l, loc=<8.7215 -0.3845 0.1059>, rot=<0.0008 0.0336 -0.0252 -0.9991>}
 clavicle_r {parent=spine_05, loc=<4.3575 -0.1466 2.3223>, rot=<0.7289 0.0465 -0.6829 0.009>}
 upperarm_r {parent=clavicle_r, loc=<-15.5111 0 0>, rot=<0.0226 -0.3234 -0.0088 -0.9459>}
 lowerarm_r {parent=upperarm_r, loc=<-26.6515 0 0>, rot=<0.0391 -0.0099 -0.2453 -0.9686>}
 hand_r {parent=lowerarm_r, loc=<-22.3492 0 0>, rot=<-0.7081 -0.0507 0.0975 -0.6975>}
 index_metacarpal_r {parent=hand_r, loc=<-1.6534 -0.0181 2.9601>, rot=<0.0158 -0.0124 -0.0173 -0.9996>}
 index_01_r {parent=index_metacarpal_r, loc=<-8.3008 -0.0357 0.0524>, rot=<-0.0044 0.0485 0.0924 -0.9945>}
 index_02_r {parent=index_01_r, loc=<-3.3589 0.3003 -0.0453>, rot=<-0.002 0.0027 0.3225 -0.9466>}
 index_03_r {parent=index_02_r, loc=<-2.5578 0.1231 0.0251>, rot=<-0.002 0.0067 0.2558 -0.9667>}
 middle_metacarpal_r {parent=hand_r, loc=<-2.1856 0.1726 0.5984>, rot=<0.006 0.0168 -0.0329 -0.9993>}
 middle_01_r {parent=middle_metacarpal_r, loc=<-7.3273 0.2447 -0.0129>, rot=<-0.0208 0.0042 0.0858 -0.9961>}
 middle_02_r {parent=middle_01_r, loc=<-4.0142 -0.1411 -0.07>, rot=<-0.0217 0.0259 0.3512 -0.9357>}
 middle_03_r {parent=middle_02_r, loc=<-2.6074 -0.0534 -0.1312>, rot=<0.0083 0.0188 0.1708 -0.9851>}
 pinky_metacarpal_r {parent=hand_r, loc=<-1.7617 0.9753 -3.3814>, rot=<-0.0397 0.0814 -0.069 -0.9935>}
 pinky_01_r {parent=pinky_metacarpal_r, loc=<-7.1707 0.4832 0.1094>, rot=<-0.0654 -0.0035 0.2242 -0.9723>}
 pinky_02_r {parent=pinky_01_r, loc=<-2.8274 -0.0892 0.1043>, rot=<0.0037 0.0011 0.2854 -0.9584>}
 pinky_03_r {parent=pinky_02_r, loc=<-2.2045 -0.031 0.0078>, rot=<-0.0002 0.0098 0.1445 -0.9895>}
 ring_metacarpal_r {parent=hand_r, loc=<-2.2073 0.3627 -1.3638>, rot=<-0.0284 0.0722 -0.0458 -0.9959>}
 ring_01_r {parent=ring_metacarpal_r, loc=<-7.0297 0.6197 0.1446>, rot=<-0.0489 -0.0252 0.1892 -0.9804>}
 ring_02_r {parent=ring_01_r, loc=<-3.6171 -0.4648 0.1027>, rot=<0 0 0.2806 -0.9598>}
 ring_03_r {parent=ring_02_r, loc=<-2.4153 -0.1141 -0.0199>, rot=<0 0 0.1189 -0.9929>}
 thumb_01_r {parent=hand_r, loc=<-1.3119 0.3192 2.459>, rot=<0.6546 -0.4042 0.0664 -0.6354>}
 thumb_02_r {parent=thumb_01_r, loc=<-5.9335 0.0447 -0.0401>, rot=<-0.0037 0.0213 0.298 -0.9543>}
 thumb_03_r {parent=thumb_02_r, loc=<-2.2223 -0.3682 0.0907>, rot=<0.0118 -0.013 0.0066 -0.9998>}
 weapon_r {parent=hand_r, loc=<0.1975 3.4112 -0.3811>, rot=<0 0 0 -1>}
 lowerarm_twist_01_r {parent=lowerarm_r, loc=<-17.855 0 0>, rot=<0 0 0 1>}
 lowerarm_twist_02_r {parent=lowerarm_r, loc=<-9.6199 0 0>, rot=<0 0 0 1>}
 dyn_ribbon_r_MSTR {parent=lowerarm_r, loc=<-0.9427 -1.5563 -4.526>, rot=<0.787 -0.2096 0.5724 0.0951>}
 dyn_ribbon_r_string_b_01 {parent=dyn_ribbon_r_MSTR, loc=<-0.5855 1.3446 -0.4098>, rot=<-0.2748 -0.7609 -0.1997 -0.5529>}
 dyn_ribbon_r_string_b_02 {parent=dyn_ribbon_r_string_b_01, loc=<6.8586 0 0>, rot=<-0.1342 -0.1911 0.0788 -0.9691>}
 dyn_ribbon_r_string_a_01 {parent=dyn_ribbon_r_MSTR, loc=<-1.5392 -0.0901 -0.9773>, rot=<-0.0551 -0.8989 -0.0266 -0.4339>}
 dyn_ribbon_r_string_a_02 {parent=dyn_ribbon_r_string_a_01, loc=<7.0438 0 0>, rot=<-0.0115 -0.0932 0.0002 -0.9956>}
 dyn_ribbon_r_loop_02 {parent=dyn_ribbon_r_MSTR, loc=<-1.5181 -0.8731 0.1642>, rot=<0.3439 -0.8961 -0.1005 0.2619>}
 dyn_ribbon_r_loop_01 {parent=dyn_ribbon_r_MSTR, loc=<-0.4819 1.0518 0.9982>, rot=<0.2547 0.6438 -0.2654 -0.671>}
 upperarm_twist_01_r {parent=upperarm_r, loc=<-0.0041 0 0>, rot=<0 0 0 1>}
 upperarm_twist_02_r {parent=upperarm_r, loc=<-13.7942 0 0>, rot=<0 0 0 1>}
 dyn_shoulder_r {parent=clavicle_r, loc=<-8.7215 0.3845 -0.1058>, rot=<0.0252 -0.9991 0.0008 -0.0336>}
 neck_01 {parent=spine_05, loc=<11.3138 0.1529 0>, rot=<0 0 -0.2003 -0.9797>}
 neck_02 {parent=neck_01, loc=<4.5627 0 0>, rot=<0 0 0.0504 -0.9987>}
 head {parent=neck_02, loc=<5.3914 0 0>, rot=<0 0 0.115 -0.9934>}
 neck_collar {parent=neck_02, loc=<0 0 0>, rot=<0 0 0 1>}
 dyn_harness_l_MSTR {parent=spine_05, loc=<-0.4899 9.6669 -11.4353>, rot=<0.0326 -0.8133 -0.5805 0.0233>}
 dyn_harness_l {parent=dyn_harness_l_MSTR, loc=<0 0 0>, rot=<0 0 0 1>}
 dyn_harness_r_MSTR {parent=spine_05, loc=<-0.5122 9.6648 11.4396>, rot=<0.0084 -0.2103 0.9768 -0.0392>}
 dyn_harness_r {parent=dyn_harness_r_MSTR, loc=<0 0 0>, rot=<0 0 0 1>}
 dyn_necklace_MSTR {parent=spine_05, loc=<12.5252 3.29 -0.0272>, rot=<0.0009 0.0004 0.9542 0.2991>}
 dyn_necklace {parent=dyn_necklace_MSTR, loc=<0 0 0>, rot=<0 0 0 1>}
 dyn_bell {parent=dyn_necklace, loc=<10.8613 0 0>, rot=<0 -0.0001 -0.044 -0.999>}
 attach_cape {parent=spine_05, loc=<9.017 -11.714 0>, rot=<0.0616 -0.7044 -0.0616 -0.7044>}
 attach_backpack {parent=spine_05, loc=<-0.123 -13.473 0>, rot=<-0.0093 -0.707 0.0093 -0.707>}
 dyn_sword_l_MSTR {parent=spine_03, loc=<3.1615 0.18 -10.919>, rot=<0.0547 -0.9985 0 0>}
 dyn_sword_l_01 {parent=dyn_sword_l_MSTR, loc=<1.431 0.2006 0.0491>, rot=<0 0 0 1>}
 dyn_sword_l_02 {parent=dyn_sword_l_01, loc=<1.951 0.3932 0.2003>, rot=<0 0 0 1>}
 dyn_sword_l_03 {parent=dyn_sword_l_02, loc=<2.5544 1.9948 0.5575>, rot=<0 0 0 1>}
 dyn_sword_r_MSTR {parent=spine_03, loc=<3.1615 0.18 10.919>, rot=<0 0 0.9985 -0.0547>}
 dyn_sword_r_01 {parent=dyn_sword_r_MSTR, loc=<1.431 -0.2006 0.0491>, rot=<0 0 0 1>}
 dyn_sword_r_02 {parent=dyn_sword_r_01, loc=<1.951 -0.3932 0.2003>, rot=<0 0 0 1>}
 dyn_sword_r_03 {parent=dyn_sword_r_02, loc=<2.5544 -1.9948 0.5575>, rot=<0 0 0 1>}
 dyn_belt_MSTR {parent=spine_02, loc=<0.3725 11.2781 4.3096>, rot=<0.4587 0.4809 0.4998 -0.5555>}
 dyn_belt {parent=dyn_belt_MSTR, loc=<0 0 0>, rot=<0 0 0 1>}
 thigh_l {parent=pelvis, loc=<-2.5693 -0.0109 -7.7501>, rot=<0.0012 0.0572 0.0155 -0.9982>}
 calf_l {parent=thigh_l, loc=<-39.0484 0 0>, rot=<0.0006 -0.0161 -0.0391 -0.9991>}
 calf_twist_01_l {parent=calf_l, loc=<-35.813 -0.0967 0.0397>, rot=<0.0237 0.0095 -0.0035 -0.9997>}
 calf_twist_02_l {parent=calf_l, loc=<-18.8064 -0.0967 0.0397>, rot=<0.0237 0.0095 -0.0035 -0.9997>}
 foot_l {parent=calf_l, loc=<-42.561 0.1077 0.4292>, rot=<0.0418 -0.0455 0.0343 -0.9975>}
 ball_l {parent=foot_l, loc=<-6.1599 13.8264 0.6806>, rot=<0.0001 -0.0001 -0.7063 -0.7079>}
 thigh_twist_01_l {parent=thigh_l, loc=<-21.6303 0 0>, rot=<0.0013 0 0 -1>}
 dyn_pants_l_front_MSTR {parent=thigh_l, loc=<-3.451 7.5669 0.0003>, rot=<0 0 0 1>}
 dyn_pants_l_front_01 {parent=dyn_pants_l_front_MSTR, loc=<0 0 0>, rot=<0 0 0 1>}
 dyn_pants_l_front_02 {parent=dyn_pants_l_front_01, loc=<-21.6303 0 0>, rot=<0.0013 0 0 -1>}
 dyn_pants_l_front_03 {parent=dyn_pants_l_front_02, loc=<-17.4181 0 0>, rot=<-0.0006 -0.016 -0.0391 -0.9991>}
 dyn_pants_l_back_MSTR {parent=thigh_l, loc=<-3.602 -5.831 0.0003>, rot=<0 0 -0.0872 -0.9962>}
 dyn_pants_l_back_01 {parent=dyn_pants_l_back_MSTR, loc=<0 0 0>, rot=<0 0 0 1>}
 dyn_pants_l_back_02 {parent=dyn_pants_l_back_01, loc=<-21.6303 0 0>, rot=<0.0013 0 0 -1>}
 dyn_pants_l_back_03 {parent=dyn_pants_l_back_02, loc=<-17.4181 0 0>, rot=<-0.0006 -0.016 -0.0391 -0.9991>}
 pants_l_back_02_reverse {parent=dyn_pants_l_back_02, loc=<-17.4181 0 0>, rot=<0 0 0 1>}
 pants_l_back_01_reverse {parent=dyn_pants_l_back_01, loc=<-21.6303 0 0>, rot=<0 0 0 1>}
 thigh_r {parent=pelvis, loc=<-2.5691 -0.0109 7.7501>, rot=<0.0572 -0.0012 0.9982 0.0155>}
 calf_r {parent=thigh_r, loc=<39.0487 0 0>, rot=<0.0006 -0.0161 -0.0391 -0.9991>}
 calf_twist_01_r {parent=calf_r, loc=<35.8132 0.0967 -0.0397>, rot=<0.0237 0.0095 -0.0035 -0.9997>}
 calf_twist_02_r {parent=calf_r, loc=<18.8064 0.0967 -0.0397>, rot=<0.0237 0.0095 -0.0035 -0.9997>}
 foot_r {parent=calf_r, loc=<42.5612 -0.1077 -0.4292>, rot=<0.0418 -0.0455 0.0343 -0.9975>}
 ball_r {parent=foot_r, loc=<6.1598 -13.8264 -0.6806>, rot=<0.0001 -0.0001 -0.7063 -0.7079>}
 thigh_twist_01_r {parent=thigh_r, loc=<21.6303 0 0>, rot=<0.0013 0 0 -1>}
 dyn_pants_r_front_MSTR {parent=thigh_r, loc=<3.4507 -7.5672 0>, rot=<0 0 0 1>}
 dyn_pants_r_front_01 {parent=dyn_pants_r_front_MSTR, loc=<0 0 0>, rot=<0 0 0 1>}
 dyn_pants_r_front_02 {parent=dyn_pants_r_front_01, loc=<21.6303 0 0>, rot=<0.0013 0 0 -1>}
 dyn_pants_r_front_03 {parent=dyn_pants_r_front_02, loc=<17.4184 0 0>, rot=<-0.0006 -0.016 -0.0391 -0.9991>}
 dyn_pants_r_back_MSTR {parent=thigh_r, loc=<3.6023 5.8314 0>, rot=<0 0 -0.0872 -0.9962>}
 dyn_pants_r_back_01 {parent=dyn_pants_r_back_MSTR, loc=<0 0 0>, rot=<0 0 0 1>}
 dyn_pants_r_back_02 {parent=dyn_pants_r_back_01, loc=<21.6303 0 0>, rot=<0.0013 0 0 -1>}
 dyn_pants_r_back_03 {parent=dyn_pants_r_back_02, loc=<17.4184 0 0>, rot=<-0.0006 -0.016 -0.0391 -0.9991>}
 pants_r_back_02_reverse {parent=dyn_pants_r_back_02, loc=<17.4184 0 0>, rot=<0 0 0 1>}
 pants_r_back_01_reverse {parent=dyn_pants_r_back_01, loc=<21.6303 0 0>, rot=<0 0 0 1>}
 ik_foot_root {parent=root, loc=<0 0 0>, rot=<0 0 0 1>}
 ik_foot_l {parent=ik_foot_root, loc=<15.2608 -4.4176 9.4433>, rot=<0.0357 0.7045 0.0284 -0.7082>}
 ik_foot_r {parent=ik_foot_root, loc=<-15.2609 -4.4176 9.4431>, rot=<0.7082 0.0284 -0.7045 0.0357>}
 ik_hand_root {parent=root, loc=<0 0 0>, rot=<0 0 0 1>}
 ik_hand_gun {parent=ik_hand_root, loc=<-51.9729 4.0883 102.6935>, rot=<0.6084 0.1793 0.4613 -0.6205>}
 ik_hand_l {parent=ik_hand_gun, loc=<53.0286 82.1782 35.2051>, rot=<-0.5102 -0.7906 -0.3387 0>}
 ik_hand_r {parent=ik_hand_gun, loc=<0 0 0>, rot=<0 0 0 1>}
 dyn_root_simspace {parent=root, loc=<0 0 0>, rot=<0 0 0 1>}
]
variables [Is DBNO {type=bool, default=False}]
