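# Physics asset for the skirt of the Kitty Warrior medium male body, covering four dyn_pants bone chains: front and back, left and right.
# Each chain hangs three capsules from a small MSTR root that rotates freely on its thigh, with near-hinge joints swinging mainly on one axis.
# Pelvis, thigh, calf and twist-bone capsules form the leg bodies.
# Leg joints are mostly free, except right calf and thigh-twist, unlike their left counterparts.
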PhysicsAsset /BRCosmetics/Characters/Player/Male/Medium/Bodies/M_MED_Kitty_Warrior/Meshes/M_MED_Kitty_Warrior_Skirt_Physics
bodies {
 calf_l [{shapeType=Capsule, length=17.1882, radius=15.198, shapeName=calf_l_capsule, center=<-17.872 -12.9209 0>, rotation=<rot 90 0 0>}]
 calf_r [{shapeType=Capsule, length=17.1882, radius=15.198, center=<17.872 12.9209 0.0003>, rotation=<rot 90 0 0>}]
 calf_twist_02_l [{shapeType=Capsule, length=16.4389, radius=13.9109, shapeName=calf_twist_02_l_capsule, center=<0 5 0>, rotation=<rot 90 0 0>}]
 calf_twist_02_r [{shapeType=Capsule, length=16.4389, radius=13.9109, shapeName=calf_twist_02_r_capsule, center=<0 -5 0>, rotation=<rot 90 0 0>}]
 dyn_pants_l_back_01 [{shapeType=Capsule, length=9.4116, radius=6.424, shapeName=dyn_pants_l_back_01_capsule, center=<-13.7897 -4 0>}]
 dyn_pants_l_back_02 [{shapeType=Capsule, length=10.3899, radius=7.5448, shapeName=dyn_pants_l_back_02_capsule, center=<-7 -4 0>}]
 dyn_pants_l_back_03 [{shapeType=Capsule, length=10.9678, radius=9.4138, shapeName=dyn_pants_l_back_03_capsule, center=<-7 -5 0>}]
 dyn_pants_l_back_MSTR [{shapeType=Capsule, length=0.505, radius=0.505, shapeName=dyn_pants_l_back_MSTR_capsule, rotation=<rot 0 -90 0>}]
 dyn_pants_l_front_01 [{shapeType=Capsule, length=6.2387, radius=7.6894, shapeName=dyn_pants_l_front_01_capsule, center=<-13 4.5 0>}]
 dyn_pants_l_front_02 [{shapeType=Capsule, length=9.951, radius=6.2, shapeName=dyn_pants_l_front_02_capsule, center=<-9 3 0>}]
 dyn_pants_l_front_03 [{shapeType=Capsule, length=12.987, radius=6.1146, shapeName=dyn_pants_l_front_03_capsule, center=<-7 1 0>}]
 dyn_pants_l_front_MSTR [{shapeType=Capsule, length=0.505, radius=0.505, shapeName=dyn_pants_l_front_MSTR_capsule, rotation=<rot 0 -90 0>}]
 dyn_pants_r_back_01 [{shapeType=Capsule, length=9.4116, radius=6.424, shapeName=dyn_pants_r_back_01_capsule, center=<13.7897 4 -0.0004>}]
 dyn_pants_r_back_02 [{shapeType=Capsule, length=10.3899, radius=7.5448, shapeName=dyn_pants_r_back_02_capsule, center=<7 4 0>}]
 dyn_pants_r_back_03 [{shapeType=Capsule, length=10.9678, radius=9.4138, shapeName=dyn_pants_r_back_03_capsule, center=<7 5 0>}]
 dyn_pants_r_back_MSTR [{shapeType=Capsule, length=0.505, radius=0.505, shapeName=dyn_pants_r_back_MSTR_capsule, rotation=<rot 0 -90 0>}]
 dyn_pants_r_front_01 [{shapeType=Capsule, length=6.2387, radius=7.6894, shapeName=dyn_pants_r_front_01_capsule, center=<13 -4.5 0>}]
 dyn_pants_r_front_02 [{shapeType=Capsule, length=9.951, radius=6.2, center=<9 -3 0>}]
 dyn_pants_r_front_03 [{shapeType=Capsule, length=12.987, radius=6.1146, shapeName=dyn_pants_r_front_03_capsule, center=<7 -1 0>}]
 dyn_pants_r_front_MSTR [{shapeType=Capsule, length=0.505, radius=0.505, shapeName=dyn_pants_r_front_MSTR_capsule, rotation=<rot 0 -90 0>}]
 pelvis [{shapeType=Capsule, length=15.3458, radius=7.3726, center=<0.2158 -2.201 0.0709>, rotation=<rot -5.798 -90 -1.0477>}]
 thigh_l [{shapeType=Capsule, length=22.1918, radius=13.7323, shapeName=thigh_l_capsule, center=<-30 7 0>, rotation=<rot 90 -90 -94.9313>}]
 thigh_r [{shapeType=Capsule, length=22.1918, radius=13.7323, shapeName=thigh_r_capsule, center=<30 -7 0>, rotation=<rot 90 1.7899 -3.1206>}]
 thigh_twist_01_l [{shapeType=Capsule, length=13.0174, radius=16.0203, shapeName=thigh_twist_01_l_capsule, center=<-0.4215 -12.085 0>, rotation=<rot 90 0 0>}]
 thigh_twist_01_r [{shapeType=Capsule, length=13.0174, radius=16.0203, center=<0.4215 12.085 -0.0002>, rotation=<rot 90 0 0>}]
}
constraints [
 {bone1Name=calf_r, bone2Name=thigh_r, swing1LimitDegrees=45, swing2LimitDegrees=45, twistLimitDegrees=45}
 {bone1Name=thigh_twist_01_r, bone2Name=thigh_r, swing1LimitDegrees=45, swing2LimitDegrees=45, twistLimitDegrees=45}
 {bone1Name=dyn_pants_r_front_02, bone2Name=dyn_pants_r_front_01, swing1LimitDegrees=45, swing2LimitDegrees=3, twistLimitDegrees=1}
 {bone1Name=dyn_pants_r_front_03, bone2Name=dyn_pants_r_front_02, swing1LimitDegrees=65, swing2LimitDegrees=3, twistLimitDegrees=1}
 {bone1Name=dyn_pants_r_front_01, bone2Name=dyn_pants_r_front_MSTR, swing1LimitDegrees=45, swing2LimitDegrees=3, twistLimitDegrees=1}
 {
  bone1Name=dyn_pants_r_front_MSTR
  bone2Name=thigh_r
  swing1LimitDegrees=45
  swing1Motion=Free
  swing2LimitDegrees=45
  swing2Motion=Free
  twistLimitDegrees=45
  twistMotion=Free
 }
 {bone1Name=dyn_pants_l_front_03, bone2Name=dyn_pants_l_front_02, swing1LimitDegrees=65, swing2LimitDegrees=3, twistLimitDegrees=1}
 {bone1Name=dyn_pants_l_front_02, bone2Name=dyn_pants_l_front_01, swing1LimitDegrees=45, swing2LimitDegrees=3, twistLimitDegrees=1}
 {bone1Name=dyn_pants_l_front_01, bone2Name=dyn_pants_l_front_MSTR, swing1LimitDegrees=45, swing2LimitDegrees=3, twistLimitDegrees=1}
 {
  bone1Name=dyn_pants_l_front_MSTR
  bone2Name=thigh_l
  swing1LimitDegrees=45
  swing1Motion=Free
  swing2LimitDegrees=45
  swing2Motion=Free
  twistLimitDegrees=45
  twistMotion=Free
 }
 {
  bone1Name=thigh_twist_01_l
  bone2Name=thigh_l
  swing1LimitDegrees=45
  swing1Motion=Free
  swing2LimitDegrees=45
  swing2Motion=Free
  twistLimitDegrees=45
  twistMotion=Free
 }
 {
  bone1Name=calf_l
  bone2Name=thigh_l
  swing1LimitDegrees=45
  swing1Motion=Free
  swing2LimitDegrees=45
  swing2Motion=Free
  twistLimitDegrees=45
  twistMotion=Free
 }
 {bone1Name=dyn_pants_r_back_03, bone2Name=dyn_pants_r_back_02, swing1LimitDegrees=45, swing2LimitDegrees=1, twistLimitDegrees=1}
 {
  bone1Name=dyn_pants_r_back_MSTR
  bone2Name=thigh_r
  swing1LimitDegrees=45
  swing1Motion=Free
  swing2LimitDegrees=45
  swing2Motion=Free
  twistLimitDegrees=45
  twistMotion=Free
 }
 {bone1Name=dyn_pants_r_back_01, bone2Name=dyn_pants_r_back_MSTR, swing1LimitDegrees=45, swing2LimitDegrees=1, twistLimitDegrees=1}
 {bone1Name=dyn_pants_r_back_02, bone2Name=dyn_pants_r_back_01, swing1LimitDegrees=45, swing2LimitDegrees=1, twistLimitDegrees=1}
 {bone1Name=dyn_pants_l_back_03, bone2Name=dyn_pants_l_back_02, swing1LimitDegrees=45, swing2LimitDegrees=1, twistLimitDegrees=1}
 {
  bone1Name=dyn_pants_l_back_MSTR
  bone2Name=thigh_l
  swing1LimitDegrees=45
  swing1Motion=Free
  swing2LimitDegrees=45
  swing2Motion=Free
  twistLimitDegrees=45
  twistMotion=Free
 }
 {bone1Name=dyn_pants_l_back_01, bone2Name=dyn_pants_l_back_MSTR, swing1LimitDegrees=45, swing2LimitDegrees=1, twistLimitDegrees=1}
 {bone1Name=dyn_pants_l_back_02, bone2Name=dyn_pants_l_back_01, swing1LimitDegrees=45, swing2LimitDegrees=1, twistLimitDegrees=1}
 {
  bone1Name=thigh_r
  bone2Name=pelvis
  swing1LimitDegrees=45
  swing1Motion=Free
  swing2LimitDegrees=45
  swing2Motion=Free
  twistLimitDegrees=45
  twistMotion=Free
 }
 {
  bone1Name=thigh_l
  bone2Name=pelvis
  swing1LimitDegrees=45
  swing1Motion=Free
  swing2LimitDegrees=45
  swing2Motion=Free
  twistLimitDegrees=45
  twistMotion=Free
 }
 {
  bone1Name=calf_twist_02_r
  bone2Name=calf_r
  swing1LimitDegrees=45
  swing1Motion=Free
  swing2LimitDegrees=45
  swing2Motion=Free
  twistLimitDegrees=45
  twistMotion=Free
 }
 {
  bone1Name=calf_twist_02_l
  bone2Name=calf_l
  swing1LimitDegrees=45
  swing1Motion=Free
  swing2LimitDegrees=45
  swing2Motion=Free
  twistLimitDegrees=45
  twistMotion=Free
 }
]
