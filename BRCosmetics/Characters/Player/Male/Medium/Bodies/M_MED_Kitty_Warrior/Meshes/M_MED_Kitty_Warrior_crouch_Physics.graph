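# Physics asset for the crouch mesh of the male medium Kitty Warrior player body, with body collision and two jointed sword chains, apparently for secondary sword motion.
# Capsules cover pelvis, spine, upper and lower arms, and thighs, with no head, hand or calf bodies.
# Each dyn_sword chain runs from a tiny MSTR body on spine_03 through short segments to a long blade capsule.
# Limb, spine_02 and sword-root joints rotate freely.
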PhysicsAsset /BRCosmetics/Characters/Player/Male/Medium/Bodies/M_MED_Kitty_Warrior/Meshes/M_MED_Kitty_Warrior_crouch_Physics
bodies {
 dyn_sword_l_01 [{shapeType=Capsule, length=3, radius=1.5, shapeName=dyn_sword_l_01_capsule, center=<1.5 0 0.5>, rotation=<rot 90 0 0>}]
 dyn_sword_l_02 [{shapeType=Sphere, radius=1.5, shapeName=dyn_sword_l_02_sphere, center=<2 0 0>}]
 dyn_sword_l_03 [{shapeType=Capsule, length=45, radius=2.2129, center=<5 0 2>, rotation=<rot 82.2773 137.5786 0>}]
 dyn_sword_l_MSTR [{shapeType=Capsule, length=0.505, radius=0.505, shapeName=dyn_sword_l_MSTR_capsule, rotation=<rot 0 -90 0>}]
 dyn_sword_r_01 [{shapeType=Capsule, length=3, radius=1.5, shapeName=dyn_sword_r_01_capsule, center=<1.5 0 0.5>, rotation=<rot 90 0 0>}]
 dyn_sword_r_02 [{shapeType=Sphere, radius=1.5, shapeName=dyn_sword_r_02_sphere, center=<1.5 0 0>}]
 dyn_sword_r_03 [{shapeType=Capsule, length=45, radius=2.2, center=<5 0 2>, rotation=<rot -81.5458 43.0028 180>}]
 dyn_sword_r_MSTR [{shapeType=Capsule, length=0.505, radius=0.505, shapeName=dyn_sword_r_MSTR_capsule, rotation=<rot 0 -90 0>}]
 lowerarm_l [{shapeType=Capsule, length=10.5, radius=4.5, shapeName=lowerarm_l_capsule, center=<9 0 0>, rotation=<rot 90 0 0>}]
 lowerarm_r [{shapeType=Capsule, length=10.5, radius=4.5, shapeName=lowerarm_r_capsule, center=<-9 0 0>, rotation=<rot 90 0 0>}]
 pelvis [{shapeType=Capsule, length=5.75, radius=11, center=<0 0.903 0>}]
 spine_01 [{shapeType=Capsule, length=2.0773, radius=10.3315, center=<0 1.2928 0.0001>}]
 spine_02 [{shapeType=Capsule, length=1.8223, radius=8.904, shapeName=spine_02_capsule, center=<0 2.5873 0>}]
 spine_03 [{shapeType=Capsule, length=0.25, radius=9.4364, center=<0 2.5786 0>}]
 spine_04 [{shapeType=Capsule, length=3.75, radius=9.6739, center=<0 2.6945 0>}]
 spine_05 [{shapeType=Capsule, length=11.2847, radius=10.8248}]
 thigh_l [{shapeType=Capsule, length=20.25, radius=10, shapeName=thigh_l_capsule, center=<-12.874 0.0004 0.7191>, rotation=<rot -82 0 0>}]
 thigh_r [{shapeType=Capsule, length=20.5, radius=9.5, shapeName=thigh_r_capsule, center=<12.874 0.0004 -0.7191>, rotation=<rot -82 0 0>}]
 upperarm_l [{shapeType=Capsule, length=8, radius=7, shapeName=upperarm_l_capsule, center=<15 0 0>, rotation=<rot 90 0 0>}]
 upperarm_r [{shapeType=Capsule, length=8, radius=7, shapeName=upperarm_r_capsule, center=<-15 0 0>, rotation=<rot 90 0 0>}]
}
constraints [
 {bone1Name=spine_01, bone2Name=pelvis, swing1LimitDegrees=45, swing2LimitDegrees=45, twistLimitDegrees=45}
 {bone1Name=spine_03, bone2Name=spine_02, swing1LimitDegrees=45, swing2LimitDegrees=45, twistLimitDegrees=45}
 {bone1Name=spine_04, bone2Name=spine_03, swing1LimitDegrees=45, swing2LimitDegrees=45, twistLimitDegrees=45}
 {bone1Name=spine_05, bone2Name=spine_04, swing1LimitDegrees=45, swing2LimitDegrees=45, twistLimitDegrees=45}
 {bone1Name=dyn_sword_l_03, bone2Name=dyn_sword_l_02, swing1LimitDegrees=60, swing2LimitDegrees=60, twistLimitDegrees=45}
 {bone1Name=dyn_sword_r_03, bone2Name=dyn_sword_r_02, swing1LimitDegrees=60, swing2LimitDegrees=60, twistLimitDegrees=45}
 {
  bone1Name=spine_02
  bone2Name=spine_01
  swing1LimitDegrees=45
  swing1Motion=Free
  swing2LimitDegrees=45
  swing2Motion=Free
  twistLimitDegrees=45
  twistMotion=Free
 }
 {bone1Name=dyn_sword_r_02, bone2Name=dyn_sword_r_01, swing1LimitDegrees=60, swing2LimitDegrees=60, twistLimitDegrees=45}
 {bone1Name=dyn_sword_r_01, bone2Name=dyn_sword_r_MSTR, swing1LimitDegrees=45, swing2LimitDegrees=45, twistLimitDegrees=35}
 {
  bone1Name=dyn_sword_r_MSTR
  bone2Name=spine_03
  swing1LimitDegrees=45
  swing1Motion=Free
  swing2LimitDegrees=45
  swing2Motion=Free
  twistLimitDegrees=45
  twistMotion=Free
 }
 {bone1Name=dyn_sword_l_02, bone2Name=dyn_sword_l_01, swing1LimitDegrees=60, swing2LimitDegrees=60, twistLimitDegrees=45}
 {bone1Name=dyn_sword_l_01, bone2Name=dyn_sword_l_MSTR, swing1LimitDegrees=60, swing2LimitDegrees=60, twistLimitDegrees=5}
 {
  bone1Name=dyn_sword_l_MSTR
  bone2Name=spine_03
  swing1LimitDegrees=45
  swing1Motion=Free
  swing2LimitDegrees=45
  swing2Motion=Free
  twistLimitDegrees=45
  twistMotion=Free
 }
 {
  bone1Name=thigh_l
  bone2Name=pelvis
  swing1LimitDegrees=45
  swing1Motion=Free
  swing2LimitDegrees=45
  swing2Motion=Free
  twistLimitDegrees=45
  twistMotion=Free
 }
 {
  bone1Name=thigh_r
  bone2Name=pelvis
  swing1LimitDegrees=45
  swing1Motion=Free
  swing2LimitDegrees=45
  swing2Motion=Free
  twistLimitDegrees=45
  twistMotion=Free
 }
 {
  bone1Name=upperarm_l
  bone2Name=spine_05
  swing1LimitDegrees=45
  swing1Motion=Free
  swing2LimitDegrees=45
  swing2Motion=Free
  twistLimitDegrees=45
  twistMotion=Free
 }
 {
  bone1Name=upperarm_r
  bone2Name=spine_05
  swing1LimitDegrees=45
  swing1Motion=Free
  swing2LimitDegrees=45
  swing2Motion=Free
  twistLimitDegrees=45
  twistMotion=Free
 }
 {
  bone1Name=lowerarm_l
  bone2Name=upperarm_l
  swing1LimitDegrees=45
  swing1Motion=Free
  swing2LimitDegrees=45
  swing2Motion=Free
  twistLimitDegrees=45
  twistMotion=Free
 }
 {
  bone1Name=lowerarm_r
  bone2Name=upperarm_r
  swing1LimitDegrees=45
  swing1Motion=Free
  swing2LimitDegrees=45
  swing2Motion=Free
  twistLimitDegrees=45
  twistMotion=Free
 }
]
