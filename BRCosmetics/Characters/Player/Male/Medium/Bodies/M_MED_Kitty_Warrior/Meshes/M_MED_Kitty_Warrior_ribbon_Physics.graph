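# Physics asset for the bow-like ribbons on both forearms of the medium male Kitty Warrior player cosmetic.
# Small master knot bodies on each lower arm anchor two separate loops and two two-segment strings, all built from capsules.
# Loops have tight twist limits, while strings swing freely at the root with limited twist and tighter-twisting tips.
# Upper spine and large offset arm capsules form the constrained body chain that carries the ribbons.
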PhysicsAsset /BRCosmetics/Characters/Player/Male/Medium/Bodies/M_MED_Kitty_Warrior/Meshes/M_MED_Kitty_Warrior_ribbon_Physics
bodies {
 dyn_ribbon_knot_l_MSTR [{shapeType=Capsule, length=0.505, radius=0.505, shapeName=dyn_ribbon_knot_l_MSTR_capsule, rotation=<rot 0 -90 0>}]
 dyn_ribbon_l_loop_01 [{shapeType=Capsule, length=2.5031, radius=3.452, shapeName=dyn_ribbon_l_loop_01_capsule, center=<5 0 0>, rotation=<rot 90 0 0>}]
 dyn_ribbon_l_loop_02 [{shapeType=Capsule, length=2.5323, radius=3.2565, shapeName=dyn_ribbon_l_loop_02_capsule, center=<5 0 0>, rotation=<rot 90 0 0>}]
 dyn_ribbon_l_string_a_01 [{shapeType=Capsule, length=1.8113, radius=1.2653, shapeName=dyn_ribbon_l_string_a_01_capsule, center=<3.5 0 0>, rotation=<rot 90 0 0>}]
 dyn_ribbon_l_string_a_02 [{shapeType=Capsule, length=6.7888, radius=1.0713, shapeName=dyn_ribbon_l_string_a_02_capsule, center=<5.4344 0 0>, rotation=<rot 90 0 0>}]
 dyn_ribbon_l_string_b_01 [{shapeType=Capsule, length=2.645, radius=1.2796, shapeName=dyn_ribbon_l_string_b_01_capsule, center=<3.5 0 0>, rotation=<rot 90 0 0>}]
 dyn_ribbon_l_string_b_02 [{shapeType=Capsule, length=5.69, radius=1.5152, shapeName=dyn_ribbon_l_string_b_02_capsule, center=<4.7125 0 0>, rotation=<rot 90 0 0>}]
 dyn_ribbon_r_MSTR [{shapeType=Capsule, length=0.505, radius=0.505, shapeName=dyn_ribbon_r_MSTR_capsule, rotation=<rot 0 -90 0>}]
 dyn_ribbon_r_loop_01 [{shapeType=Capsule, length=0.2567, radius=4.5759, shapeName=dyn_ribbon_r_loop_01_capsule, center=<5.3727 0 0>, rotation=<rot 90 0 0>}]
 dyn_ribbon_r_loop_02 [{shapeType=Capsule, length=2.5322, radius=3.2515, shapeName=dyn_ribbon_r_loop_02_capsule, center=<5.4492 0 0>, rotation=<rot 90 0 0>}]
 dyn_ribbon_r_string_a_01 [{shapeType=Capsule, length=3.395, radius=1.2796, shapeName=dyn_ribbon_r_string_a_01_capsule, center=<3.5 0 0>, rotation=<rot 90 0 0>}]
 dyn_ribbon_r_string_a_02 [{shapeType=Capsule, length=6.19, radius=1.2652, shapeName=dyn_ribbon_r_string_a_02_capsule, center=<4.7125 0 0>, rotation=<rot 90 0 0>}]
 dyn_ribbon_r_string_b_01 [{shapeType=Capsule, length=3.1393, radius=1.4374, shapeName=dyn_ribbon_r_string_b_01_capsule, center=<4 0 0>, rotation=<rot 90 0 0>}]
 dyn_ribbon_r_string_b_02 [{shapeType=Capsule, length=6.2888, radius=1.8213, shapeName=dyn_ribbon_r_string_b_02_capsule, center=<5.4344 0 0>, rotation=<rot 90 0 0>}]
 lowerarm_l [{shapeType=Capsule, length=6.25, radius=14.25, shapeName=lowerarm_l_capsule, center=<9 1.4392 -9.7805>, rotation=<rot 90 0 0>}]
 lowerarm_r [{shapeType=Capsule, length=7, radius=12.25, shapeName=lowerarm_r_capsule, center=<-8.9999 -2.3576 7.7425>, rotation=<rot 90 0 0>}]
 spine_03 [{shapeType=Capsule, length=0.5, radius=9.9364, center=<0 3.3885 0>}]
 spine_04 [{shapeType=Capsule, length=5.25, radius=9.6739, center=<0 2.6945 0>}]
 spine_05 [{shapeType=Capsule, length=11.2847, radius=10.8248}]
 upperarm_l [{shapeType=Capsule, length=2.25, radius=14.25, shapeName=upperarm_l_capsule, center=<15 -2.4432 -5.1154>, rotation=<rot 90 0 0>}]
 upperarm_r [{shapeType=Capsule, length=2.25, radius=13.25, shapeName=upperarm_r_capsule, center=<-15 0.0437 3.89>, rotation=<rot 90 0 0>}]
}
constraints [
 {bone1Name=spine_04, bone2Name=spine_03, swing1LimitDegrees=45, swing2LimitDegrees=45, twistLimitDegrees=45}
 {bone1Name=spine_05, bone2Name=spine_04, swing1LimitDegrees=45, swing2LimitDegrees=45, twistLimitDegrees=45}
 {
  bone1Name=upperarm_l
  bone2Name=spine_05
  swing1LimitDegrees=45
  swing1Motion=Free
  swing2LimitDegrees=45
  swing2Motion=Free
  twistLimitDegrees=45
  twistMotion=Free
 }
 {
  bone1Name=upperarm_r
  bone2Name=spine_05
  swing1LimitDegrees=45
  swing1Motion=Free
  swing2LimitDegrees=45
  swing2Motion=Free
  twistLimitDegrees=45
  twistMotion=Free
 }
 {
  bone1Name=lowerarm_l
  bone2Name=upperarm_l
  swing1LimitDegrees=45
  swing1Motion=Free
  swing2LimitDegrees=45
  swing2Motion=Free
  twistLimitDegrees=45
  twistMotion=Free
 }
 {
  bone1Name=lowerarm_r
  bone2Name=upperarm_r
  swing1LimitDegrees=45
  swing1Motion=Free
  swing2LimitDegrees=45
  swing2Motion=Free
  twistLimitDegrees=45
  twistMotion=Free
 }
 {
  bone1Name=dyn_ribbon_r_MSTR
  bone2Name=lowerarm_r
  swing1LimitDegrees=45
  swing1Motion=Free
  swing2LimitDegrees=45
  swing2Motion=Free
  twistLimitDegrees=45
  twistMotion=Free
 }
 {
  bone1Name=dyn_ribbon_r_string_b_01
  bone2Name=dyn_ribbon_r_MSTR
  swing1LimitDegrees=30
  swing1Motion=Free
  swing2LimitDegrees=30
  swing2Motion=Free
  twistLimitDegrees=15
 }
 {bone1Name=dyn_ribbon_r_string_b_02, bone2Name=dyn_ribbon_r_string_b_01, swing1LimitDegrees=45, swing2LimitDegrees=45, twistLimitDegrees=1}
 {
  bone1Name=dyn_ribbon_r_string_a_01
  bone2Name=dyn_ribbon_r_MSTR
  swing1LimitDegrees=30
  swing1Motion=Free
  swing2LimitDegrees=30
  swing2Motion=Free
  twistLimitDegrees=15
 }
 {bone1Name=dyn_ribbon_r_string_a_02, bone2Name=dyn_ribbon_r_string_a_01, swing1LimitDegrees=45, swing2LimitDegrees=45, twistLimitDegrees=1}
 {bone1Name=dyn_ribbon_r_loop_02, bone2Name=dyn_ribbon_r_MSTR, swing1LimitDegrees=60, swing2LimitDegrees=60, twistLimitDegrees=5}
 {bone1Name=dyn_ribbon_r_loop_01, bone2Name=dyn_ribbon_r_MSTR, swing1LimitDegrees=60, swing2LimitDegrees=60, twistLimitDegrees=5}
 {bone1Name=dyn_ribbon_l_loop_01, bone2Name=dyn_ribbon_knot_l_MSTR, swing1LimitDegrees=60, swing2LimitDegrees=60, twistLimitDegrees=5}
 {
  bone1Name=dyn_ribbon_knot_l_MSTR
  bone2Name=lowerarm_l
  swing1LimitDegrees=45
  swing1Motion=Free
  swing2LimitDegrees=45
  swing2Motion=Free
  twistLimitDegrees=45
  twistMotion=Free
 }
 {bone1Name=dyn_ribbon_l_loop_02, bone2Name=dyn_ribbon_knot_l_MSTR, swing1LimitDegrees=60, swing2LimitDegrees=60, twistLimitDegrees=5}
 {
  bone1Name=dyn_ribbon_l_string_b_01
  bone2Name=dyn_ribbon_knot_l_MSTR
  swing1LimitDegrees=30
  swing1Motion=Free
  swing2LimitDegrees=60
  swing2Motion=Free
  twistLimitDegrees=15
 }
 {bone1Name=dyn_ribbon_l_string_b_02, bone2Name=dyn_ribbon_l_string_b_01, swing1LimitDegrees=45, swing2LimitDegrees=45, twistLimitDegrees=5}
 {
  bone1Name=dyn_ribbon_l_string_a_01
  bone2Name=dyn_ribbon_knot_l_MSTR
  swing1LimitDegrees=30
  swing1Motion=Free
  swing2LimitDegrees=60
  swing2Motion=Free
  twistLimitDegrees=15
 }
 {bone1Name=dyn_ribbon_l_string_a_02, bone2Name=dyn_ribbon_l_string_a_01, swing1LimitDegrees=45, swing2LimitDegrees=45, twistLimitDegrees=5}
]
