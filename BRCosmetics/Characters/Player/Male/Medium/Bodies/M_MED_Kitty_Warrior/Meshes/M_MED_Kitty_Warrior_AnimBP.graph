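# Post-process style AnimBlueprint for the Kitty Warrior body cosmetic, copying the parent mesh pose and animating its ribbons, harness, belt, swords and pants.
# Weighted constraints and bone offsets feed a cached pose, which branches into RigidBody setups for front end, crouch, skydive and regular play.
# Variable-driven leg constraints correct the simulated pants.
# A sitting-emote cache receives additive pants offsets, followed outside the front end by a DBNO-aware Control Rig.
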AnimBlueprint /BRCosmetics/Characters/Player/Male/Medium/Bodies/M_MED_Kitty_Warrior/Meshes/M_MED_Kitty_Warrior_AnimBP
target_skeleton=&/M_MED_Kitty_Warrior_Skeleton
graphs [
 AnimGraph {nodes [
   Component To Local()
   n1 = "Constraint - dyn_pants_l_front_01 (P)"(Alpha=1, ConstraintWeights_0=1) {
    alphaBoolBlend {blendOption=Linear}
    boneToModify {boneName=dyn_pants_l_front_01}
    constraintSetup [{targetBone {boneName=thigh_l}}]
    constraintWeights [1]
   }
   n2 = "Constraint - dyn_pants_l_front_02 (PP)"!(ComponentPose=@n1, Alpha=1, ConstraintWeights_0=0.25, ConstraintWeights_1:float=0.75) -> (Pose) {
    alphaBoolBlend {blendOption=Linear}
    boneToModify {boneName=dyn_pants_l_front_02}
    constraintSetup [{targetBone {boneName=calf_l}}, {targetBone {boneName=thigh_l}}]
    constraintWeights [1,1]
   }
   "Constraint - dyn_pants_l_front_03 (P)"(ComponentPose=@n2, Alpha=1, ConstraintWeights_0=1) {
    alphaBoolBlend {blendOption=Linear}
    boneToModify {boneName=dyn_pants_l_front_03}
    constraintSetup [{targetBone {boneName=calf_twist_02_l}}]
    constraintWeights [1]
   }
   n3 = "Constraint - dyn_pants_r_front_01 (P)"(Alpha=1, ConstraintWeights_0=1) {
    alphaBoolBlend {blendOption=Linear}
    boneToModify {boneName=dyn_pants_r_front_01}
    constraintSetup [{targetBone {boneName=thigh_r}}]
    constraintWeights [1]
   }
   n4 = "Constraint - dyn_pants_r_front_02 (PP)"!(ComponentPose=@n3, Alpha=1, ConstraintWeights_0=0.25, ConstraintWeights_1:float=0.75) -> (Pose) {
    alphaBoolBlend {blendOption=Linear}
    boneToModify {boneName=dyn_pants_r_front_02}
    constraintSetup [{targetBone {boneName=calf_r}}, {targetBone {boneName=thigh_r}}]
    constraintWeights [1,1]
   }
   "Constraint - dyn_pants_r_front_03 (P)"(ComponentPose=@n4, Alpha=1, ConstraintWeights_0=1) {
    alphaBoolBlend {blendOption=Linear}
    boneToModify {boneName=dyn_pants_r_front_03}
    constraintSetup [{targetBone {boneName=calf_twist_02_r}}]
    constraintWeights [1]
   }
   "Constraint - dyn_pants_r_front_02 (P)"(Alpha=1, ConstraintWeights_0=1) {
    alphaBoolBlend {blendOption=Linear}
    boneToModify {boneName=dyn_pants_r_front_02}
    constraintSetup [{targetBone {boneName=thigh_twist_01_r}}]
    constraintWeights [1]
   }
   n5 = "Constraint - dyn_pants_l_front_MSTR (P)"(Alpha=1, ConstraintWeights_0=1) {
    alphaBoolBlend {blendOption=Linear}
    boneToModify {boneName=dyn_pants_l_front_MSTR}
    constraintSetup [{targetBone {boneName=thigh_l}}]
    constraintWeights [1]
   }
   "Constraint - dyn_pants_l_back_MSTR (P)"(ComponentPose=@n5, Alpha=1, ConstraintWeights_0=1) {
    alphaBoolBlend {blendOption=Linear}
    boneToModify {boneName=dyn_pants_l_back_MSTR}
    constraintSetup [{targetBone {boneName=thigh_l}}]
    constraintWeights [1]
   }
   Local To Component()
   n6 = Get GravityOverride()
   RigidBody!(ComponentPose, Alpha, OverrideWorldGravity:Vector=@n6, ExternalForce) -> (Pose) {
    bClampLinearTranslationLimitToRefPose=true
    bOverrideWorldGravity=true
    overridePhysicsAsset=&/M_MED_Kitty_Warrior_Skirt_Physics
    simulationSpace=BaseBoneSpace
    alphaBoolBlend {blendInTime=0.1, blendOption=Linear, blendOutTime=0.1}
    baseBoneRef {boneName=root}
    componentAppliedLinearAccClamp=<3000 3000 3000>
    componentLinearAccScale=<0.3 0.3 0.3>
    componentLinearVelScale=<0.3 0.3 0.3>
    overrideWorldGravity {z=-980}
    simSpaceSettings {
     maxAngularAcceleration=200
     maxAngularVelocity=5
     maxLinearAcceleration=1000
     maxLinearVelocity=300
     velocityScaleZ=0
     worldAlpha=1
     externalLinearDragV=<1 1>
    }
   }
   n7 = "Transform (Modify) Bone - Bone: dyn_pants_r_front_MSTR"(
    ComponentPose=Local To Component(LocalPose=Use cached pose 'CRcache_sittingEmotes'())
    Alpha=1
    Translation="0,-5,0"
    Rotation="0,0,50"
    Scale="1,1,1"
   ) {
    rotationMode=BMM_Additive
    rotationSpace=BCS_BoneSpace
    scaleMode=BMM_Additive
    scaleSpace=BCS_BoneSpace
    translationMode=BMM_Additive
    translationSpace=BCS_BoneSpace
    alphaBoolBlend {blendOption=Linear}
    alphaScaleBiasClamp {bInterpResult=true, interpSpeedDecreasing=2, interpSpeedIncreasing=2}
    boneToModify {boneName=dyn_pants_r_front_MSTR}
    rotation {roll=50}
    translation {y=-5}
   }
   n8 = "Transform (Modify) Bone - Bone: dyn_pants_l_front_02"(ComponentPose=@n7, Alpha=1, Translation="0,5,0", Rotation="0,20,30", Scale="1,1,1") {
    rotationMode=BMM_Additive
    rotationSpace=BCS_BoneSpace
    scaleMode=BMM_Additive
    scaleSpace=BCS_BoneSpace
    translationMode=BMM_Additive
    translationSpace=BCS_BoneSpace
    alphaBoolBlend {blendOption=Linear}
    alphaScaleBiasClamp {bInterpResult=true, interpSpeedDecreasing=2, interpSpeedIncreasing=2}
    boneToModify {boneName=dyn_pants_l_front_02}
    rotation {roll=30, yaw=20}
    translation {y=5}
   }
   n9 = "Transform (Modify) Bone - Bone: dyn_pants_r_front_02"(ComponentPose=@n8, Alpha=1, Translation="0,-5,0", Rotation="25,20,30", Scale="1,1,1") {
    rotationMode=BMM_Additive
    rotationSpace=BCS_BoneSpace
    scaleMode=BMM_Additive
    scaleSpace=BCS_BoneSpace
    translationMode=BMM_Additive
    translationSpace=BCS_BoneSpace
    alphaBoolBlend {blendOption=Linear}
    alphaScaleBiasClamp {bInterpResult=true, interpSpeedDecreasing=2, interpSpeedIncreasing=2}
    boneToModify {boneName=dyn_pants_r_front_02}
    rotation=<rot 25 20 30>
    translation {y=-5}
   }
   n10 = "Transform (Modify) Bone - Bone: dyn_pants_r_front_03"(ComponentPose=@n9, Alpha=1, Translation="0,0,0", Rotation="30,0,0", Scale="1,1,1") {
    rotationMode=BMM_Additive
    rotationSpace=BCS_BoneSpace
    scaleMode=BMM_Additive
    scaleSpace=BCS_BoneSpace
    translationMode=BMM_Additive
    translationSpace=BCS_BoneSpace
    alphaBoolBlend {blendOption=Linear}
    alphaScaleBiasClamp {bInterpResult=true, interpSpeedDecreasing=2, interpSpeedIncreasing=2}
    boneToModify {boneName=dyn_pants_r_front_03}
    rotation {pitch=30}
   }
   n11 = "Transform (Modify) Bone - Bone: dyn_pants_l_front_03"(ComponentPose=@n10, Alpha=1, Translation="0,0,0", Rotation="20,50,0", Scale="1,1,1") {
    rotationMode=BMM_Additive
    rotationSpace=BCS_BoneSpace
    scaleMode=BMM_Additive
    scaleSpace=BCS_BoneSpace
    translationMode=BMM_Additive
    translationSpace=BCS_BoneSpace
    alphaBoolBlend {blendOption=Linear}
    alphaScaleBiasClamp {bInterpResult=true, interpSpeedDecreasing=2, interpSpeedIncreasing=2}
    boneToModify {boneName=dyn_pants_l_front_03}
    rotation {pitch=20, yaw=50}
   }
   n12 = "Transform (Modify) Bone - Bone: dyn_pants_l_front_02"(ComponentPose=@n11, Alpha=1, Translation="0,0,0", Rotation="0,15,0", Scale="1,1,1") {
    rotationMode=BMM_Additive
    rotationSpace=BCS_BoneSpace
    scaleMode=BMM_Additive
    translationMode=BMM_Additive
    translationSpace=BCS_BoneSpace
    alphaBoolBlend {blendOption=Linear}
    alphaScaleBiasClamp {bInterpResult=true}
    boneToModify {boneName=dyn_pants_l_front_02}
    rotation {yaw=15}
   }
   n13 = "Transform (Modify) Bone - Bone: dyn_pants_l_front_03"(ComponentPose=@n12, Alpha=1, Translation="0,0,0", Rotation="0,60,0", Scale="1,1,1") {
    rotationMode=BMM_Additive
    rotationSpace=BCS_BoneSpace
    scaleMode=BMM_Additive
    translationMode=BMM_Additive
    translationSpace=BCS_BoneSpace
    alphaBoolBlend {blendOption=Linear}
    alphaScaleBiasClamp {bInterpResult=true}
    boneToModify {boneName=dyn_pants_l_front_03}
    rotation {yaw=60}
   }
   n14 = "Transform (Modify) Bone - Bone: dyn_pants_r_front_03"(ComponentPose=@n13, Alpha=1, Translation="0,0,0", Rotation="0,60,0", Scale="1,1,1") {
    rotationMode=BMM_Additive
    rotationSpace=BCS_BoneSpace
    scaleMode=BMM_Additive
    translationMode=BMM_Additive
    translationSpace=BCS_BoneSpace
    alphaBoolBlend {blendOption=Linear}
    alphaScaleBiasClamp {bInterpResult=true}
    boneToModify {boneName=dyn_pants_r_front_03}
    rotation {yaw=60}
   }
   n15 = "Transform (Modify) Bone - Bone: dyn_pants_l_front_02"(ComponentPose=@n14, Alpha=1, Translation="0,0,0", Rotation="0,15,0", Scale="1,1,1") {
    rotationMode=BMM_Additive
    rotationSpace=BCS_BoneSpace
    scaleMode=BMM_Additive
    translationMode=BMM_Additive
    translationSpace=BCS_BoneSpace
    alphaBoolBlend {blendOption=Linear}
    alphaScaleBiasClamp {bInterpResult=true}
    boneToModify {boneName=dyn_pants_l_front_02}
    rotation {yaw=15}
   }
   n16 = "Transform (Modify) Bone - Bone: dyn_pants_l_front_03"(ComponentPose=@n15, Alpha=1, Translation="0,0,0", Rotation="0,60,0", Scale="1,1,1") {
    rotationMode=BMM_Additive
    rotationSpace=BCS_BoneSpace
    scaleMode=BMM_Additive
    translationMode=BMM_Additive
    translationSpace=BCS_BoneSpace
    alphaBoolBlend {blendOption=Linear}
    alphaScaleBiasClamp {bInterpResult=true}
    boneToModify {boneName=dyn_pants_l_front_03}
    rotation {yaw=60}
   }
   n17 = "Transform (Modify) Bone - Bone: dyn_pants_r_front_02"(ComponentPose=@n16, Alpha=1, Translation="0,0,0", Rotation="0,15,0", Scale="1,1,1") {
    rotationMode=BMM_Additive
    rotationSpace=BCS_BoneSpace
    scaleMode=BMM_Additive
    translationMode=BMM_Additive
    translationSpace=BCS_BoneSpace
    alphaBoolBlend {blendOption=Linear}
    alphaScaleBiasClamp {bInterpResult=true}
    boneToModify {boneName=dyn_pants_r_front_02}
    rotation {yaw=15}
   }
   n18 = "Transform (Modify) Bone - Bone: dyn_pants_r_front_03"(ComponentPose=@n17, Alpha=1, Translation="0,0,0", Rotation="0,60,0", Scale="1,1,1") {
    rotationMode=BMM_Additive
    rotationSpace=BCS_BoneSpace
    scaleMode=BMM_Additive
    translationMode=BMM_Additive
    translationSpace=BCS_BoneSpace
    alphaBoolBlend {blendOption=Linear}
    alphaScaleBiasClamp {bInterpResult=true}
    boneToModify {boneName=dyn_pants_r_front_03}
    rotation {yaw=60}
   }
   n19 = "Transform (Modify) Bone - Bone: dyn_pants_r_front_MSTR"(
    ComponentPose=Local To Component(LocalPose=Use cached pose 'CRcache_sittingEmotes'())
    Alpha=1
    Translation="0,-5,0"
    Rotation="0,0,50"
    Scale="1,1,1"
   ) {
    rotationMode=BMM_Additive
    rotationSpace=BCS_BoneSpace
    scaleMode=BMM_Additive
    scaleSpace=BCS_BoneSpace
    translationMode=BMM_Additive
    translationSpace=BCS_BoneSpace
    alphaBoolBlend {blendOption=Linear}
    alphaScaleBiasClamp {bInterpResult=true, interpSpeedDecreasing=2, interpSpeedIncreasing=2}
    boneToModify {boneName=dyn_pants_r_front_MSTR}
    rotation {roll=50}
    translation {y=-5}
   }
   n20 = "Transform (Modify) Bone - Bone: dyn_pants_l_front_02"(ComponentPose=@n19, Alpha=1, Translation="0,5,0", Rotation="0,20,30", Scale="1,1,1") {
    rotationMode=BMM_Additive
    rotationSpace=BCS_BoneSpace
    scaleMode=BMM_Additive
    scaleSpace=BCS_BoneSpace
    translationMode=BMM_Additive
    translationSpace=BCS_BoneSpace
    alphaBoolBlend {blendOption=Linear}
    alphaScaleBiasClamp {bInterpResult=true, interpSpeedDecreasing=2, interpSpeedIncreasing=2}
    boneToModify {boneName=dyn_pants_l_front_02}
    rotation {roll=30, yaw=20}
    translation {y=5}
   }
   n21 = "Transform (Modify) Bone - Bone: dyn_pants_r_front_02"(ComponentPose=@n20, Alpha=1, Translation="0,-5,0", Rotation="25,20,30", Scale="1,1,1") {
    rotationMode=BMM_Additive
    rotationSpace=BCS_BoneSpace
    scaleMode=BMM_Additive
    scaleSpace=BCS_BoneSpace
    translationMode=BMM_Additive
    translationSpace=BCS_BoneSpace
    alphaBoolBlend {blendOption=Linear}
    alphaScaleBiasClamp {bInterpResult=true, interpSpeedDecreasing=2, interpSpeedIncreasing=2}
    boneToModify {boneName=dyn_pants_r_front_02}
    rotation=<rot 25 20 30>
    translation {y=-5}
   }
   n22 = "Transform (Modify) Bone - Bone: dyn_pants_r_front_03"(ComponentPose=@n21, Alpha=1, Translation="0,0,0", Rotation="30,0,0", Scale="1,1,1") {
    rotationMode=BMM_Additive
    rotationSpace=BCS_BoneSpace
    scaleMode=BMM_Additive
    scaleSpace=BCS_BoneSpace
    translationMode=BMM_Additive
    translationSpace=BCS_BoneSpace
    alphaBoolBlend {blendOption=Linear}
    alphaScaleBiasClamp {bInterpResult=true, interpSpeedDecreasing=2, interpSpeedIncreasing=2}
    boneToModify {boneName=dyn_pants_r_front_03}
    rotation {pitch=30}
   }
   n23 = "Transform (Modify) Bone - Bone: dyn_pants_l_front_03"(ComponentPose=@n22, Alpha=1, Translation="0,0,0", Rotation="20,50,0", Scale="1,1,1") {
    rotationMode=BMM_Additive
    rotationSpace=BCS_BoneSpace
    scaleMode=BMM_Additive
    scaleSpace=BCS_BoneSpace
    translationMode=BMM_Additive
    translationSpace=BCS_BoneSpace
    alphaBoolBlend {blendOption=Linear}
    alphaScaleBiasClamp {bInterpResult=true, interpSpeedDecreasing=2, interpSpeedIncreasing=2}
    boneToModify {boneName=dyn_pants_l_front_03}
    rotation {pitch=20, yaw=50}
   }
   n24 = "Transform (Modify) Bone - Bone: dyn_pants_l_front_02"(ComponentPose=@n23, Alpha=1, Translation="0,0,0", Rotation="0,15,0", Scale="1,1,1") {
    rotationMode=BMM_Additive
    rotationSpace=BCS_BoneSpace
    scaleMode=BMM_Additive
    translationMode=BMM_Additive
    translationSpace=BCS_BoneSpace
    alphaBoolBlend {blendOption=Linear}
    alphaScaleBiasClamp {bInterpResult=true}
    boneToModify {boneName=dyn_pants_l_front_02}
    rotation {yaw=15}
   }
   n25 = "Transform (Modify) Bone - Bone: dyn_pants_l_front_03"(ComponentPose=@n24, Alpha=1, Translation="0,0,0", Rotation="0,60,0", Scale="1,1,1") {
    rotationMode=BMM_Additive
    rotationSpace=BCS_BoneSpace
    scaleMode=BMM_Additive
    translationMode=BMM_Additive
    translationSpace=BCS_BoneSpace
    alphaBoolBlend {blendOption=Linear}
    alphaScaleBiasClamp {bInterpResult=true}
    boneToModify {boneName=dyn_pants_l_front_03}
    rotation {yaw=60}
   }
   n26 = "Transform (Modify) Bone - Bone: dyn_pants_r_front_03"(ComponentPose=@n25, Alpha=1, Translation="0,0,0", Rotation="0,60,0", Scale="1,1,1") {
    rotationMode=BMM_Additive
    rotationSpace=BCS_BoneSpace
    scaleMode=BMM_Additive
    translationMode=BMM_Additive
    translationSpace=BCS_BoneSpace
    alphaBoolBlend {blendOption=Linear}
    alphaScaleBiasClamp {bInterpResult=true}
    boneToModify {boneName=dyn_pants_r_front_03}
    rotation {yaw=60}
   }
   n27 = "Transform (Modify) Bone - Bone: dyn_pants_l_front_02"(ComponentPose=@n26, Alpha=1, Translation="0,0,0", Rotation="0,15,0", Scale="1,1,1") {
    rotationMode=BMM_Additive
    rotationSpace=BCS_BoneSpace
    scaleMode=BMM_Additive
    translationMode=BMM_Additive
    translationSpace=BCS_BoneSpace
    alphaBoolBlend {blendOption=Linear}
    alphaScaleBiasClamp {bInterpResult=true}
    boneToModify {boneName=dyn_pants_l_front_02}
    rotation {yaw=15}
   }
   n28 = "Transform (Modify) Bone - Bone: dyn_pants_l_front_03"(ComponentPose=@n27, Alpha=1, Translation="0,0,0", Rotation="0,60,0", Scale="1,1,1") {
    rotationMode=BMM_Additive
    rotationSpace=BCS_BoneSpace
    scaleMode=BMM_Additive
    translationMode=BMM_Additive
    translationSpace=BCS_BoneSpace
    alphaBoolBlend {blendOption=Linear}
    alphaScaleBiasClamp {bInterpResult=true}
    boneToModify {boneName=dyn_pants_l_front_03}
    rotation {yaw=60}
   }
   n29 = "Transform (Modify) Bone - Bone: dyn_pants_r_front_02"(ComponentPose=@n28, Alpha=1, Translation="0,0,0", Rotation="0,15,0", Scale="1,1,1") {
    rotationMode=BMM_Additive
    rotationSpace=BCS_BoneSpace
    scaleMode=BMM_Additive
    translationMode=BMM_Additive
    translationSpace=BCS_BoneSpace
    alphaBoolBlend {blendOption=Linear}
    alphaScaleBiasClamp {bInterpResult=true}
    boneToModify {boneName=dyn_pants_r_front_02}
    rotation {yaw=15}
   }
   n30 = "Transform (Modify) Bone - Bone: dyn_pants_r_front_03"(ComponentPose=@n29, Alpha=1, Translation="0,0,0", Rotation="0,60,0", Scale="1,1,1") {
    rotationMode=BMM_Additive
    rotationSpace=BCS_BoneSpace
    scaleMode=BMM_Additive
    translationMode=BMM_Additive
    translationSpace=BCS_BoneSpace
    alphaBoolBlend {blendOption=Linear}
    alphaScaleBiasClamp {bInterpResult=true}
    boneToModify {boneName=dyn_pants_r_front_03}
    rotation {yaw=60}
   }
   n31 = Control Rig!(Source=Component To Local(ComponentPose=@n30), Alpha, Is DBNO:bool=Get bIsDBNO()) -> (Pose) {
    alphaBoolBlend {blendOption=Linear}
    controlRigAssetReference {blueprintRigClass=&/M_MED_Kitty_Warrior_CtrlRig.M_MED_Kitty_Warrior_CtrlRig_C}
   }
   Output Pose(Result=Blend Poses by bool(BlendPose_0=Component To Local(ComponentPose=@n18), BlendPose_1=@n31, bActiveValue=Get bIsInFrontEnd()) {blendType=Linear}) {layerGroup=None}
   n32 = "Constraint - dyn_ribbon_r_MSTR (PPPP)"!(
    ComponentPose=Local To Component(LocalPose=Copy Pose From Mesh() {bUseAttachedParent=true})
    Alpha=1
    ConstraintWeights_0=0.3085
    ConstraintWeights_1:float=0.3127
    ConstraintWeights_2:float=0.3501
    ConstraintWeights_3:float=0.0287
   ) -> (Pose) {
    alphaBoolBlend {blendOption=Linear}
    boneToModify {boneName=dyn_ribbon_r_MSTR}
    constraintSetup [
     {targetBone {boneName=upperarm_r}}
     {targetBone {boneName=lowerarm_r}}
     {targetBone {boneName=lowerarm_twist_02_r}}
     {targetBone {boneName=upperarm_twist_02_r}}
    ]
    constraintWeights [1,1,1,1]
   }
   n33 = "Constraint - dyn_ribbon_knot_l_MSTR (PPPP)"!(
    ComponentPose=@n32
    Alpha=1
    ConstraintWeights_0=0.3085
    ConstraintWeights_1:float=0.3127
    ConstraintWeights_2:float=0.3501
    ConstraintWeights_3:float=0.0287
   ) -> (Pose) {
    alphaBoolBlend {blendOption=Linear}
    boneToModify {boneName=dyn_ribbon_knot_l_MSTR}
    constraintSetup [
     {targetBone {boneName=upperarm_l}}
     {targetBone {boneName=lowerarm_l}}
     {targetBone {boneName=lowerarm_twist_02_l}}
     {targetBone {boneName=upperarm_twist_02_l}}
    ]
    constraintWeights [1,1,1,1]
   }
   n34 = "Constraint - dyn_harness_l_MSTR (PPPPPPP)"!(
    ComponentPose=@n33
    Alpha=1
    ConstraintWeights_0=0.0711
    ConstraintWeights_1:float=0.1458
    ConstraintWeights_2:float=0.3833
    ConstraintWeights_3:float=0.3433
    ConstraintWeights_4:float=0.0341
    ConstraintWeights_5:float=0.0037
    ConstraintWeights_6:float=0.0187
   ) -> (Pose) {
    alphaBoolBlend {blendOption=Linear}
    boneToModify {boneName=dyn_harness_l_MSTR}
    constraintSetup [
     {targetBone {boneName=spine_03}}
     {targetBone {boneName=spine_04}}
     {targetBone {boneName=spine_05}}
     {targetBone {boneName=clavicle_l}}
     {targetBone {boneName=upperarm_twist_01_l}}
     {targetBone {boneName=neck_01}}
     {targetBone {boneName=neck_02}}
    ]
    constraintWeights [1,1,1,1,1,1,1]
   }
   n35 = "Constraint - dyn_harness_r_MSTR (PPPPPPPP)"!(
    ComponentPose=@n34
    Alpha=1
    ConstraintWeights_0=0.0714
    ConstraintWeights_1:float=0.1383
    ConstraintWeights_2:float=0.3941
    ConstraintWeights_3:float=0.3458
    ConstraintWeights_4:float=0.0276
    ConstraintWeights_5:float=0.0029
    ConstraintWeights_6:float=0.02
    ConstraintWeights_7:float=0
   ) -> (Pose) {
    alphaBoolBlend {blendOption=Linear}
    boneToModify {boneName=dyn_harness_r_MSTR}
    constraintSetup [
     {targetBone {boneName=spine_03}}
     {targetBone {boneName=spine_04}}
     {targetBone {boneName=spine_05}}
     {targetBone {boneName=clavicle_r}}
     {targetBone {boneName=upperarm_twist_01_r}}
     {targetBone {boneName=neck_01}}
     {targetBone {boneName=neck_02}}
     {targetBone {boneName=head}}
    ]
    constraintWeights [1,1,1,1,1,1,1,1]
   }
   n36 = "Constraint - dyn_belt_MSTR (PPPPP)"!(
    ComponentPose=@n35
    Alpha=1
    ConstraintWeights_0=0.1765
    ConstraintWeights_1:float=0.2558
    ConstraintWeights_2:float=0.2934
    ConstraintWeights_3:float=0.2195
    ConstraintWeights_4:float=0.0548
   ) -> (Pose) {
    alphaBoolBlend {blendOption=Linear}
    boneToModify {boneName=dyn_belt_MSTR}
    constraintSetup [
     {targetBone {boneName=pelvis}}
     {targetBone {boneName=spine_01}}
     {targetBone {boneName=spine_02}}
     {targetBone {boneName=spine_03}}
     {targetBone {boneName=spine_04}}
    ]
    constraintWeights [1,1,1,1,1]
   }
   n37 = "Constraint - dyn_sword_r_MSTR (PPPPPPPP)"!(
    ComponentPose=@n36
    Alpha=1
    ConstraintWeights_0=0.017
    ConstraintWeights_1:float=0.2254
    ConstraintWeights_2:float=0.4254
    ConstraintWeights_3:float=0.3319
    ConstraintWeights_4:float=0.0003
    ConstraintWeights_5:float=0
    ConstraintWeights_6:float=0
    ConstraintWeights_7:float=0
   ) -> (Pose) {
    alphaBoolBlend {blendOption=Linear}
    boneToModify {boneName=dyn_sword_r_MSTR}
    constraintSetup [
     {targetBone {boneName=spine_01}}
     {targetBone {boneName=spine_02}}
     {targetBone {boneName=spine_03}}
     {targetBone {boneName=spine_04}}
     {targetBone {boneName=spine_05}}
     {targetBone {boneName=clavicle_r}}
     {targetBone {boneName=upperarm_twist_01_r}}
     {targetBone {boneName=neck_02}}
    ]
    constraintWeights [1,1,1,1,1,1,1,1]
   }
   n38 = "Constraint - dyn_sword_l_MSTR (PPPPPPPP)"!(
    ComponentPose=@n37
    Alpha=1
    ConstraintWeights_0=0.0039
    ConstraintWeights_1:float=0.0348
    ConstraintWeights_2:float=0.2394
    ConstraintWeights_3:float=0.4107
    ConstraintWeights_4:float=0.3101
    ConstraintWeights_5:float=0.001
    ConstraintWeights_6:float=0
    ConstraintWeights_7:float=0
   ) -> (Pose) {
    alphaBoolBlend {blendOption=Linear}
    boneToModify {boneName=dyn_sword_l_MSTR}
    constraintSetup [
     {targetBone {boneName=pelvis}}
     {targetBone {boneName=spine_01}}
     {targetBone {boneName=spine_02}}
     {targetBone {boneName=spine_03}}
     {targetBone {boneName=spine_04}}
     {targetBone {boneName=spine_05}}
     {targetBone {boneName=clavicle_l}}
     {targetBone {boneName=upperarm_twist_01_l}}
    ]
    constraintWeights [1,1,1,1,1,1,1,1]
   }
   n39 = "Constraint - dyn_pants_r_front_MSTR (P)"(ComponentPose=@n38, Alpha=1, ConstraintWeights_0=1) {
    alphaBoolBlend {blendOption=Linear}
    boneToModify {boneName=dyn_pants_r_front_MSTR}
    constraintSetup [{targetBone {boneName=thigh_r}}]
    constraintWeights [1]
   }
   n40 = "Constraint - dyn_pants_r_back_MSTR (P)"(ComponentPose=@n39, Alpha=1, ConstraintWeights_0=1) {
    alphaBoolBlend {blendOption=Linear}
    boneToModify {boneName=dyn_pants_r_back_MSTR}
    constraintSetup [{targetBone {boneName=thigh_r}}]
    constraintWeights [1]
   }
   n41 = "Constraint - dyn_pants_l_front_MSTR (P)"(ComponentPose=@n40, Alpha=1, ConstraintWeights_0=1) {
    alphaBoolBlend {blendOption=Linear}
    boneToModify {boneName=dyn_pants_l_front_MSTR}
    constraintSetup [{targetBone {boneName=thigh_l}}]
    constraintWeights [1]
   }
   n42 = "Constraint - dyn_pants_l_back_MSTR (P)"(ComponentPose=@n41, Alpha=1, ConstraintWeights_0=1) {
    alphaBoolBlend {blendOption=Linear}
    boneToModify {boneName=dyn_pants_l_back_MSTR}
    constraintSetup [{targetBone {boneName=thigh_l}}]
    constraintWeights [1]
   }
   n43 = "Transform (Modify) Bone - Bone: dyn_shoulder_l"(ComponentPose=@n42, Alpha=Get Upperarm l() -> (Upperarm l:double), Translation="10,0,-2", Rotation="-17,0,0", Scale="1,1,1") {
    rotationMode=BMM_Additive
    rotationSpace=BCS_BoneSpace
    translationMode=BMM_Additive
    translationSpace=BCS_BoneSpace
    alphaBoolBlend {blendOption=Linear}
    boneToModify {boneName=dyn_shoulder_l}
    rotation {pitch=-17}
    translation {x=10, z=-2}
   }
   n44 = "Transform (Modify) Bone - Bone: dyn_shoulder_r"(ComponentPose=@n43, Alpha=Get Upperarm r() -> (Upperarm r:double), Translation="-10,0,2", Rotation="17,0,0", Scale="1,1,1") {
    rotationMode=BMM_Additive
    rotationSpace=BCS_BoneSpace
    translationMode=BMM_Additive
    translationSpace=BCS_BoneSpace
    alphaBoolBlend {blendOption=Linear}
    boneToModify {boneName=dyn_shoulder_r}
    rotation {pitch=17}
    translation {x=-10, z=2}
   }
   Save cached pose 'pre rigid'(Pose=Component To Local(ComponentPose=@n44)) {cacheName=pre rigid}
   n45 = "Transform (Modify) Bone - Bone: dyn_sword_l_MSTR"(
    ComponentPose=Local To Component(LocalPose=Use cached pose 'pre rigid'())
    Alpha=1
    Translation="-4.581,-0.613,1.718"
    Rotation="0,0.000,-15"
    Scale="1,1,1"
   ) {
    rotationMode=BMM_Additive
    rotationSpace=BCS_BoneSpace
    translationMode=BMM_Additive
    translationSpace=BCS_BoneSpace
    alphaBoolBlend {blendOption=Linear}
    boneToModify {boneName=dyn_sword_l_MSTR}
    rotation {roll=-15}
    translation=<-4.581 -0.6133 1.7181>
   }
   n46 = Get GravityOverride()
   n47 = RigidBody(ComponentPose=@n45, OverrideWorldGravity=@n46) {
    bClampLinearTranslationLimitToRefPose=true
    bOverrideWorldGravity=true
    overridePhysicsAsset=&/M_MED_Kitty_Warrior_crouch_Physics
    simulationSpace=BaseBoneSpace
    alphaBoolBlend {blendOption=Linear}
    baseBoneRef {boneName=spine_04}
    componentAppliedLinearAccClamp=<3000 3000 3000>
    componentLinearAccScale=<0.25 0.25 0.2>
    componentLinearVelScale=<0.2 0.2 0.2>
   }
   n48 = RigidBody(ComponentPose=@n47, OverrideWorldGravity=@n46) {
    bClampLinearTranslationLimitToRefPose=true
    bOverrideWorldGravity=true
    overridePhysicsAsset=&M_MED_Kitty_Warrior_Skirt_Physics
    simulationSpace=BaseBoneSpace
    alphaBoolBlend {blendOption=Linear}
    baseBoneRef {boneName=root}
    componentAppliedLinearAccClamp=<3000 3000 3000>
    componentLinearAccScale=<0.25 0.25 0.4>
    componentLinearVelScale=<0.2 0.2 0.2>
   }
   n49 = Get thigh_rot_r()
   n50 = "Constraint - dyn_pants_r_front_01 (P)"(ComponentPose=@n48, Alpha=@n49, ConstraintWeights_0=1) {
    alphaBoolBlend {blendOption=Linear}
    boneToModify {boneName=dyn_pants_r_front_01}
    constraintSetup [{targetBone {boneName=thigh_r}}]
    constraintWeights [1]
   }
   n51 = "Constraint - dyn_pants_r_front_02 (PP)"(ComponentPose=@n50, Alpha=@n49, ConstraintWeights_0=0.25, ConstraintWeights_1=0.75) {
    alphaBoolBlend {blendOption=Linear}
    boneToModify {boneName=dyn_pants_r_front_02}
    constraintSetup [{targetBone {boneName=calf_r}}, {targetBone {boneName=thigh_r}}]
    constraintWeights [1,1]
   }
   n52 = Get calf_rot_r() -> (calf_rot_r:double)
   n53 = "Constraint - dyn_pants_r_front_03 (P)"(ComponentPose=@n51, Alpha=@n52, ConstraintWeights_0=1) {
    alphaBoolBlend {blendOption=Linear}
    boneToModify {boneName=dyn_pants_r_front_03}
    constraintSetup [{targetBone {boneName=calf_twist_02_r}}]
    constraintWeights [1]
   }
   n54 = "Constraint - dyn_pants_r_back_01 (PT)"!(ComponentPose=@n53, Alpha=@n52, ConstraintWeights_0=1, ConstraintWeights_1:float=0) -> (Pose) {
    alphaBoolBlend {blendOption=Linear}
    boneToModify {boneName=dyn_pants_r_back_01}
    constraintSetup [{targetBone {boneName=thigh_r}}, {transformType=Translation, targetBone {boneName=calf_r}}]
    constraintWeights [1,1]
   }
   n55 = "Constraint - dyn_pants_r_back_02 (PT)"!(ComponentPose=@n54, Alpha=@n52, ConstraintWeights_0=0.5, ConstraintWeights_1:float=0.5) -> (Pose) {
    alphaBoolBlend {blendOption=Linear}
    boneToModify {boneName=dyn_pants_r_back_02}
    constraintSetup [{targetBone {boneName=calf_r}}, {transformType=Translation, targetBone {boneName=thigh_r}}]
    constraintWeights [1,1]
   }
   n56 = "Constraint - dyn_pants_r_back_03 (PP)"!(ComponentPose=@n55, Alpha=@n52, ConstraintWeights_0=0.9, ConstraintWeights_1:float=0.1) -> (Pose) {
    alphaBoolBlend {blendOption=Linear}
    boneToModify {boneName=dyn_pants_r_back_03}
    constraintSetup [{targetBone {boneName=calf_r}}, {targetBone {boneName=thigh_r}}]
    constraintWeights [1,1]
   }
   n57 = Get thigh_rot_l()
   n58 = "Constraint - dyn_pants_l_front_01 (P)"(ComponentPose=@n56, Alpha=@n57, ConstraintWeights_0=1) {
    alphaBoolBlend {blendOption=Linear}
    boneToModify {boneName=dyn_pants_l_front_01}
    constraintSetup [{targetBone {boneName=thigh_l}}]
    constraintWeights [1]
   }
   n59 = "Constraint - dyn_pants_l_front_02 (PP)"(ComponentPose=@n58, Alpha=@n57, ConstraintWeights_0=0.25, ConstraintWeights_1=0.75) {
    alphaBoolBlend {blendOption=Linear}
    boneToModify {boneName=dyn_pants_l_front_02}
    constraintSetup [{targetBone {boneName=calf_l}}, {targetBone {boneName=thigh_l}}]
    constraintWeights [1,1]
   }
   n60 = Get calf_rot_l() -> (calf_rot_l:double)
   n61 = "Constraint - dyn_pants_l_front_03 (P)"(ComponentPose=@n59, Alpha=@n60, ConstraintWeights_0=1) {
    alphaBoolBlend {blendOption=Linear}
    boneToModify {boneName=dyn_pants_l_front_03}
    constraintSetup [{targetBone {boneName=calf_twist_02_l}}]
    constraintWeights [1]
   }
   n62 = "Constraint - dyn_pants_l_back_01 (P)"(ComponentPose=@n61, Alpha=@n57, ConstraintWeights_0=1) {
    alphaBoolBlend {blendOption=Linear}
    boneToModify {boneName=dyn_pants_l_back_01}
    constraintSetup [{targetBone {boneName=thigh_l}}]
    constraintWeights [1]
   }
   n63 = "Constraint - dyn_pants_l_back_02 (PT)"!(ComponentPose=@n62, Alpha=@n57, ConstraintWeights_0=0.5, ConstraintWeights_1:float=0.5) -> (Pose) {
    alphaBoolBlend {blendOption=Linear}
    boneToModify {boneName=dyn_pants_l_back_02}
    constraintSetup [{targetBone {boneName=calf_l}}, {transformType=Translation, targetBone {boneName=thigh_l}}]
    constraintWeights [1,1]
   }
   n64 = "Constraint - dyn_pants_l_back_03 (PP)"!(ComponentPose=@n63, Alpha=@n60, ConstraintWeights_0=0.8, ConstraintWeights_1:float=0.2) -> (Pose) {
    alphaBoolBlend {blendOption=Linear}
    boneToModify {boneName=dyn_pants_l_back_03}
    constraintSetup [{targetBone {boneName=calf_l}}, {targetBone {boneName=thigh_l}}]
    constraintWeights [1,1]
   }
   Save cached pose 'crouch'(Pose=Component To Local(ComponentPose=@n64)) {cacheName=crouch}
   n65 = Get GravityOverride()
   n66 = RigidBody(ComponentPose=Local To Component(LocalPose=Use cached pose 'pre rigid'()), OverrideWorldGravity=@n65) {
    bClampLinearTranslationLimitToRefPose=true
    bOverrideWorldGravity=true
    overridePhysicsAsset=&/M_MED_Kitty_Warrior_FrontEnd_Physics
    simulationSpace=BaseBoneSpace
    alphaBoolBlend {blendOption=Linear}
    baseBoneRef {boneName=spine_02}
    componentAppliedLinearAccClamp=<300 300 300>
    componentLinearAccScale=<0.25 0.25 0.25>
    componentLinearVelScale=<0.2 0.2 0.2>
   }
   n67 = RigidBody(ComponentPose=@n66, Alpha=Get skirt rban() -> (skirt rban:double), OverrideWorldGravity=@n65) {
    bClampLinearTranslationLimitToRefPose=true
    bOverrideWorldGravity=true
    overridePhysicsAsset=&M_MED_Kitty_Warrior_Skirt_Physics
    simulationSpace=BaseBoneSpace
    alphaBoolBlend {blendOption=Linear}
    baseBoneRef {boneName=root}
    componentAppliedLinearAccClamp=<3000 3000 3000>
    componentLinearAccScale=<0.25 0.25 0.4>
    componentLinearVelScale=<0.2 0.2 0.2>
   }
   n68 = Get calf_rot_r()
   n69 = "Constraint - dyn_pants_r_back_01 (PT)"(ComponentPose=@n67, Alpha=@n68, ConstraintWeights_0=1, ConstraintWeights_1=0) {
    alphaBoolBlend {blendOption=Linear}
    boneToModify {boneName=dyn_pants_r_back_01}
    constraintSetup [{targetBone {boneName=thigh_r}}, {transformType=Translation, targetBone {boneName=calf_r}}]
    constraintWeights [1,1]
   }
   n70 = "Constraint - dyn_pants_r_back_02 (PT)"(ComponentPose=@n69, Alpha=@n68, ConstraintWeights_0=1, ConstraintWeights_1=0) {
    alphaBoolBlend {blendOption=Linear}
    boneToModify {boneName=dyn_pants_r_back_02}
    constraintSetup [{targetBone {boneName=calf_r}}, {transformType=Translation, targetBone {boneName=thigh_r}}]
    constraintWeights [1,1]
   }
   n71 = "Constraint - dyn_pants_r_back_03 (PP)"(ComponentPose=@n70, Alpha=@n68, ConstraintWeights_0=0.75, ConstraintWeights_1=0.25) {
    alphaBoolBlend {blendOption=Linear}
    boneToModify {boneName=dyn_pants_r_back_03}
    constraintSetup [{targetBone {boneName=calf_r}}, {targetBone {boneName=thigh_r}}]
    constraintWeights [1,1]
   }
   n72 = Get calf_rot_l()
   n73 = "Constraint - dyn_pants_l_back_01 (P)"(ComponentPose=@n71, Alpha=@n72, ConstraintWeights_0=1) {
    alphaBoolBlend {blendOption=Linear}
    boneToModify {boneName=dyn_pants_l_back_01}
    constraintSetup [{targetBone {boneName=thigh_l}}]
    constraintWeights [1]
   }
   n74 = "Constraint - dyn_pants_l_back_02 (PT)"(ComponentPose=@n73, Alpha=@n72, ConstraintWeights_0=1, ConstraintWeights_1=0) {
    alphaBoolBlend {blendOption=Linear}
    boneToModify {boneName=dyn_pants_l_back_02}
    constraintSetup [{targetBone {boneName=calf_l}}, {transformType=Translation, targetBone {boneName=thigh_l}}]
    constraintWeights [1,1]
   }
   n75 = "Constraint - dyn_pants_l_back_03 (PP)"(ComponentPose=@n74, Alpha=@n72, ConstraintWeights_0=1, ConstraintWeights_1=0) {
    alphaBoolBlend {blendOption=Linear}
    boneToModify {boneName=dyn_pants_l_back_03}
    constraintSetup [{targetBone {boneName=calf_l}}, {targetBone {boneName=thigh_l}}]
    constraintWeights [1,1]
   }
   Save cached pose 'front end'(Pose=Component To Local(ComponentPose=@n75)) {cacheName=front end}
   n76 = Get calf_rot_r()
   n77 = "Transform (Modify) Bone - Bone: dyn_pants_r_back_01"(
    ComponentPose=Local To Component(LocalPose=Use cached pose 'crouch'())
    Alpha=@n76
    Translation="-5,-3,0.000"
    Rotation="0,0,0"
    Scale="1,1,1"
   ) {
    rotationMode=BMM_Additive
    rotationSpace=BCS_BoneSpace
    translationMode=BMM_Additive
    translationSpace=BCS_BoneSpace
    alphaBoolBlend {blendOption=Linear}
    boneToModify {boneName=dyn_pants_r_back_01}
    translation=<-5 -3>
   }
   n78 = "Transform (Modify) Bone - Bone: dyn_pants_r_back_02"(ComponentPose=@n77, Alpha=@n76, Translation="0,0,0", Rotation="-0.000,-10,0.000", Scale="1,1,1") {
    rotationMode=BMM_Additive
    rotationSpace=BCS_BoneSpace
    translationMode=BMM_Additive
    translationSpace=BCS_BoneSpace
    alphaBoolBlend {blendOption=Linear}
    boneToModify {boneName=dyn_pants_r_back_02}
    rotation {yaw=-10}
   }
   n79 = "Transform (Modify) Bone - Bone: dyn_pants_r_back_03"(ComponentPose=@n78, Alpha=@n76, Translation="6,-3,2", Rotation="0,-20,0", Scale="1,1,1") {
    rotationMode=BMM_Additive
    rotationSpace=BCS_BoneSpace
    translationMode=BMM_Additive
    translationSpace=BCS_BoneSpace
    alphaBoolBlend {blendOption=Linear}
    boneToModify {boneName=dyn_pants_r_back_03}
    rotation {yaw=-20}
    translation=<6 -3 2>
   }
   n80 = Component To Local(ComponentPose=@n79)
   n81 = Use cached pose 'crouch'()
   n82 = Get bIsMoving2D()
   n83 = RigidBody(ComponentPose=Local To Component(LocalPose=Use cached pose 'pre rigid'()), OverrideWorldGravity=@n6) {
    bClampLinearTranslationLimitToRefPose=true
    bOverrideWorldGravity=true
    overridePhysicsAsset=&/M_MED_Kitty_Warrior_Physics
    simulationSpace=BaseBoneSpace
    alphaBoolBlend {blendOption=Linear}
    baseBoneRef {boneName=spine_02}
    componentAppliedLinearAccClamp=<1000 1000 1000>
    componentLinearAccScale=<0.25 0.25 0.25>
    componentLinearVelScale=<0.2 0.2 0.2>
   }
   n84 = RigidBody(ComponentPose=@n83, Alpha=Get skirt rban(), OverrideWorldGravity=@n6) {
    bClampLinearTranslationLimitToRefPose=true
    bOverrideWorldGravity=true
    overridePhysicsAsset=&M_MED_Kitty_Warrior_Skirt_Physics
    simulationSpace=BaseBoneSpace
    alphaBoolBlend {blendOption=Linear}
    baseBoneRef {boneName=root}
    componentAppliedLinearAccClamp=<3000 3000 3000>
    componentLinearAccScale=<0.25 0.25 0.4>
    componentLinearVelScale=<0.2 0.2 0.2>
   }
   n85 = Get calf_rot_r()
   n86 = "Constraint - dyn_pants_r_back_01 (PT)"(ComponentPose=@n84, Alpha=@n85, ConstraintWeights_0=1, ConstraintWeights_1=0) {
    alphaBoolBlend {blendOption=Linear}
    boneToModify {boneName=dyn_pants_r_back_01}
    constraintSetup [{targetBone {boneName=thigh_r}}, {transformType=Translation, targetBone {boneName=calf_r}}]
    constraintWeights [1,1]
   }
   n87 = "Constraint - dyn_pants_r_back_02 (PT)"(ComponentPose=@n86, Alpha=@n85, ConstraintWeights_0=1, ConstraintWeights_1=0) {
    alphaBoolBlend {blendOption=Linear}
    boneToModify {boneName=dyn_pants_r_back_02}
    constraintSetup [{targetBone {boneName=calf_r}}, {transformType=Translation, targetBone {boneName=thigh_r}}]
    constraintWeights [1,1]
   }
   n88 = "Constraint - dyn_pants_r_back_03 (PP)"(ComponentPose=@n87, Alpha=@n85, ConstraintWeights_0=0.75, ConstraintWeights_1=0.25) {
    alphaBoolBlend {blendOption=Linear}
    boneToModify {boneName=dyn_pants_r_back_03}
    constraintSetup [{targetBone {boneName=calf_r}}, {targetBone {boneName=thigh_r}}]
    constraintWeights [1,1]
   }
   n89 = Get calf_rot_l()
   n90 = "Constraint - dyn_pants_l_back_01 (P)"(ComponentPose=@n88, Alpha=@n89, ConstraintWeights_0=1) {
    alphaBoolBlend {blendOption=Linear}
    boneToModify {boneName=dyn_pants_l_back_01}
    constraintSetup [{targetBone {boneName=thigh_l}}]
    constraintWeights [1]
   }
   n91 = "Constraint - dyn_pants_l_back_02 (PT)"(ComponentPose=@n90, Alpha=@n89, ConstraintWeights_0=1, ConstraintWeights_1=0) {
    alphaBoolBlend {blendOption=Linear}
    boneToModify {boneName=dyn_pants_l_back_02}
    constraintSetup [{targetBone {boneName=calf_l}}, {transformType=Translation, targetBone {boneName=thigh_l}}]
    constraintWeights [1,1]
   }
   n92 = "Constraint - dyn_pants_l_back_03 (PP)"(ComponentPose=@n91, Alpha=@n89, ConstraintWeights_0=1, ConstraintWeights_1=0) {
    alphaBoolBlend {blendOption=Linear}
    boneToModify {boneName=dyn_pants_l_back_03}
    constraintSetup [{targetBone {boneName=calf_l}}, {targetBone {boneName=thigh_l}}]
    constraintWeights [1,1]
   }
   n93 = Blend Poses by bool(
    BlendPose_0=Use cached pose 'front end'()
    BlendPose_1=Blend Poses by bool(
     BlendPose_0=Blend Poses by bool(BlendPose_0=@n81, BlendPose_1=@n80, bActiveValue=@n82) {blendType=Linear}
     BlendPose_1=Component To Local(ComponentPose=@n92)
     BlendTime_1=0.3
     bActiveValue=Get bIsCrouching()
    ) {blendType=Linear}
    bActiveValue=Get bIsInFrontEnd()
   ) {blendType=Linear}
   n94 = "Transform (Modify) Bone - Bone: dyn_sword_l_MSTR"(
    ComponentPose=Local To Component(LocalPose=Use cached pose 'pre rigid'())
    Alpha=1
    Translation="0,0,0"
    Rotation="35,0.000,0.000"
    Scale="1,1,1"
   ) {
    rotationMode=BMM_Additive
    rotationSpace=BCS_BoneSpace
    alphaBoolBlend {blendOption=Linear}
    boneToModify {boneName=dyn_sword_l_MSTR}
    rotation {pitch=35}
   }
   n95 = "Transform (Modify) Bone - Bone: dyn_sword_r_MSTR"(ComponentPose=@n94, Alpha=1, Translation="0,0,0", Rotation="35,0,0", Scale="1,1,1") {
    rotationMode=BMM_Additive
    rotationSpace=BCS_BoneSpace
    alphaBoolBlend {blendOption=Linear}
    boneToModify {boneName=dyn_sword_r_MSTR}
    rotation {pitch=35}
   }
   n96 = Get GravityOverride()
   n97 = RigidBody(ComponentPose=@n95, OverrideWorldGravity=@n96) {
    bClampLinearTranslationLimitToRefPose=true
    bOverrideWorldGravity=true
    overridePhysicsAsset=&/M_MED_Kitty_Warrior_skydive_Physics
    simulationSpace=BaseBoneSpace
    alphaBoolBlend {blendOption=Linear}
    baseBoneRef {boneName=pelvis}
    componentAppliedLinearAccClamp=<3000 3000 3000>
    componentLinearAccScale=<0.25 0.25 0.4>
    componentLinearVelScale=<0.2 0.2 0.2>
   }
   n98 = RigidBody(ComponentPose=@n97, Alpha=Get skirt rban(), OverrideWorldGravity=@n96) {
    bClampLinearTranslationLimitToRefPose=true
    bOverrideWorldGravity=true
    overridePhysicsAsset=&M_MED_Kitty_Warrior_Skirt_Physics
    simulationSpace=BaseBoneSpace
    alphaBoolBlend {blendOption=Linear}
    baseBoneRef {boneName=root}
    componentAppliedLinearAccClamp=<3000 3000 3000>
    componentLinearAccScale=<0.25 0.25 0.4>
    componentLinearVelScale=<0.2 0.2 0.2>
   }
   n99 = Get calf_rot_r()
   n100 = "Constraint - dyn_pants_r_back_01 (PT)"(ComponentPose=@n98, Alpha=@n99, ConstraintWeights_0=1, ConstraintWeights_1=0) {
    alphaBoolBlend {blendOption=Linear}
    boneToModify {boneName=dyn_pants_r_back_01}
    constraintSetup [{targetBone {boneName=thigh_r}}, {transformType=Translation, targetBone {boneName=calf_r}}]
    constraintWeights [1,1]
   }
   n101 = "Constraint - dyn_pants_r_back_02 (PT)"(ComponentPose=@n100, Alpha=@n99, ConstraintWeights_0=1, ConstraintWeights_1=0) {
    alphaBoolBlend {blendOption=Linear}
    boneToModify {boneName=dyn_pants_r_back_02}
    constraintSetup [{targetBone {boneName=calf_r}}, {transformType=Translation, targetBone {boneName=thigh_r}}]
    constraintWeights [1,1]
   }
   n102 = "Constraint - dyn_pants_r_back_03 (PP)"(ComponentPose=@n101, Alpha=@n99, ConstraintWeights_0=0.75, ConstraintWeights_1=0.25) {
    alphaBoolBlend {blendOption=Linear}
    boneToModify {boneName=dyn_pants_r_back_03}
    constraintSetup [{targetBone {boneName=calf_r}}, {targetBone {boneName=thigh_r}}]
    constraintWeights [1,1]
   }
   n103 = Get calf_rot_l()
   n104 = "Constraint - dyn_pants_l_back_01 (P)"(ComponentPose=@n102, Alpha=@n103, ConstraintWeights_0=1) {
    alphaBoolBlend {blendOption=Linear}
    boneToModify {boneName=dyn_pants_l_back_01}
    constraintSetup [{targetBone {boneName=thigh_l}}]
    constraintWeights [1]
   }
   n105 = "Constraint - dyn_pants_l_back_02 (PT)"(ComponentPose=@n104, Alpha=@n103, ConstraintWeights_0=1, ConstraintWeights_1=0) {
    alphaBoolBlend {blendOption=Linear}
    boneToModify {boneName=dyn_pants_l_back_02}
    constraintSetup [{targetBone {boneName=calf_l}}, {transformType=Translation, targetBone {boneName=thigh_l}}]
    constraintWeights [1,1]
   }
   n106 = "Constraint - dyn_pants_l_back_03 (PP)"(ComponentPose=@n105, Alpha=@n103, ConstraintWeights_0=1, ConstraintWeights_1=0) {
    alphaBoolBlend {blendOption=Linear}
    boneToModify {boneName=dyn_pants_l_back_03}
    constraintSetup [{targetBone {boneName=calf_l}}, {targetBone {boneName=thigh_l}}]
    constraintWeights [1,1]
   }
   n107 = Component To Local(ComponentPose=@n106)
   n108 = Get bIsSkyDiving()
   n109 = RigidBody(
    ComponentPose=Local To Component(LocalPose=Blend Poses by bool(BlendPose_0=@n107, BlendPose_1=@n93, bActiveValue=@n108) {blendType=Linear})
    OverrideWorldGravity=Get GravityOverride()
   ) {
    bClampLinearTranslationLimitToRefPose=true
    bOverrideWorldGravity=true
    overridePhysicsAsset=&/M_MED_Kitty_Warrior_ribbon_Physics
    simulationSpace=BaseBoneSpace
    alphaBoolBlend {blendOption=Linear}
    baseBoneRef {boneName=spine_02}
    componentAppliedLinearAccClamp=<3000 3000 3000>
    componentLinearAccScale=<0.25 0.25 0.4>
    componentLinearVelScale=<0.2 0.2 0.2>
   }
   n110 = Anim Dynamics - Bone: dyn_belt(ComponentPose=@n109, Alpha=1, ExternalForce="0,0,0") {
    angularDampingOverride=0.9999
    angularSpringConstant=100
    bAngularSpring=true
    bGravityOverrideInSimSpace=true
    bOverrideAngularDamping=true
    simulationSpace=RootRelative
    alphaBoolBlend {blendOption=Linear}
    boundBone {boneName=dyn_belt}
    componentAppliedLinearAccClamp=<3500 3500 3500>
    componentLinearAccScale=<1.2 1.2 1.2>
    componentLinearVelScale=<1.2 1.2 1.2>
    physicsBodyDefinitions [{
      sphereCollisionRadius=0
      boundBone {boneName=dyn_belt}
      boxExtents=<40 30 30>
      constraintSetup {angularLimitsMax=<1 10 15>, angularLimitsMin=<-1 -30 -15>, angularTarget=<0 0 0>}
      localJointOffset=<2 0 0>
     }]
    relativeSpaceBone {boneName=dyn_root_simspace}
   }
   n111 = Anim Dynamics - Bone: dyn_harness_l(ComponentPose=@n110, Alpha=1, ExternalForce="0,0,0") {
    angularDampingOverride=0.9999
    angularSpringConstant=100
    bAngularSpring=true
    bGravityOverrideInSimSpace=true
    bOverrideAngularDamping=true
    simulationSpace=RootRelative
    alphaBoolBlend {blendOption=Linear}
    boundBone {boneName=dyn_harness_l}
    componentAppliedLinearAccClamp=<3500 3500 3500>
    componentLinearAccScale=<1.2 1.2 1.2>
    componentLinearVelScale=<1.2 1.2 1.2>
    physicsBodyDefinitions [{
      sphereCollisionRadius=0
      boundBone {boneName=dyn_harness_l}
      boxExtents=<40 30 30>
      constraintSetup {angularLimitsMax=<1 10 15>, angularLimitsMin=<-1 -30 -15>, angularTarget=<0 0 0>}
      localJointOffset=<2 0 0>
     }]
    relativeSpaceBone {boneName=dyn_root_simspace}
   }
   n112 = Anim Dynamics - Bone: dyn_harness_r(ComponentPose=@n111, Alpha=1, ExternalForce="0,0,0") {
    angularDampingOverride=0.9999
    angularSpringConstant=100
    bAngularSpring=true
    bGravityOverrideInSimSpace=true
    bOverrideAngularDamping=true
    simulationSpace=RootRelative
    alphaBoolBlend {blendOption=Linear}
    boundBone {boneName=dyn_harness_r}
    componentAppliedLinearAccClamp=<3500 3500 3500>
    componentLinearAccScale=<1.2 1.2 1.2>
    componentLinearVelScale=<1.2 1.2 1.2>
    physicsBodyDefinitions [{
      sphereCollisionRadius=0
      boundBone {boneName=dyn_harness_r}
      boxExtents=<40 30 30>
      constraintSetup {angularLimitsMax=<1 26 15>, angularLimitsMin=<-1 -15 -30>, angularTarget=<0 0 0>}
      localJointOffset=<2 0 0>
     }]
    relativeSpaceBone {boneName=dyn_root_simspace}
   }
   n113 = Anim Dynamics - Bone: dyn_necklace(ComponentPose=@n112, Alpha=1, ExternalForce="0,0,0") {
    angularDampingOverride=0.998
    angularSpringConstant=70
    bAngularSpring=true
    bGravityOverrideInSimSpace=true
    bOverrideAngularDamping=true
    bUseGravityOverride=true
    bUsePlanarLimit=false
    simulationSpace=RootRelative
    alphaBoolBlend {blendOption=Linear}
    boundBone {boneName=dyn_necklace}
    componentAppliedLinearAccClamp=<3000 3000 3000>
    componentLinearAccScale=<0.3 0.3 0.3>
    componentLinearVelScale=<0.3 0.3 0.3>
    gravityOverride {z=-980}
    physicsBodyDefinitions [{
      collisionType=InnerSphere
      sphereCollisionRadius=1
      boundBone {boneName=dyn_necklace}
      boxExtents=<7 12 12>
      constraintSetup {twistAxis=AxisZ, angularLimitsMax=<0 0 0>, angularLimitsMin=<0 0 -5>, angularTarget=<0 0 0>}
      localJointOffset=<2 0 0>
     }]
    relativeSpaceBone {boneName=root}
   }
   n114 = Anim Dynamics - Bone: dyn_bell(ComponentPose=@n113, Alpha=1, ExternalForce="0,0,0") {
    angularDampingOverride=0.998
    angularSpringConstant=70
    bAngularSpring=true
    bGravityOverrideInSimSpace=true
    bOverrideAngularDamping=true
    bUseGravityOverride=true
    bUsePlanarLimit=false
    simulationSpace=RootRelative
    alphaBoolBlend {blendOption=Linear}
    boundBone {boneName=dyn_bell}
    componentAppliedLinearAccClamp=<3000 3000 3000>
    componentLinearAccScale=<0.3 0.3 0.3>
    componentLinearVelScale=<0.3 0.3 0.3>
    gravityOverride {z=-980}
    physicsBodyDefinitions [{
      collisionType=InnerSphere
      sphereCollisionRadius=1
      boundBone {boneName=dyn_bell}
      boxExtents=<7 12 12>
      constraintSetup {twistAxis=AxisZ, angularLimitsMax=<0 5 0>, angularLimitsMin=<0 -5 -35>, angularTarget=<0 0 0>}
      localJointOffset=<2 0 0>
     }]
    relativeSpaceBone {boneName=root}
    sphericalLimits [
     {limitRadius=7, drivingBone {boneName=spine_04}, sphereLocalOffset=<0 3.1 4>}
     {limitRadius=7, drivingBone {boneName=spine_04}, sphereLocalOffset=<0 3.1 -4>}
    ]
   }
   Save cached pose 'CRcache_sittingEmotes'(Pose=Component To Local(ComponentPose=@n114)) {cacheName=CRcache_sittingEmotes}
   Get thigh_rot_r()
   Get calf_rot_r()
   Get thigh_rot_r()
   Get thigh_rot_r()
   Get thigh_rot_l()
   Get calf_rot_r()
   Get thigh_rot_r()
   Get thigh_rot_l()
   Get thigh_rot_r()
   Get thigh_rot_l()
  ]}
 EventGraph {nodes [
   Event BlueprintUpdateAnimation()
   > n1 = Sequence!(execute) -> (then_0, then_1, then_2:Exec, then_3:Exec, then_4:Exec, then_5:Exec, then_6:Exec, then_7:Exec, then_8:Exec)
   > n2 = Branch(Condition=Get bIsCrouching())
   n3 = Get Skel_Mesh()
   > Set thigh_rot_l(thigh_rot_l=GetAlphaFromReferenceTransformDelta(SkeletalMeshComponent=@n3, BoneName=thigh_l, BaseName=pelvis, InputRange=(X=-20,Y=-100), OutputRange=(X=0,Y=0.8), TwistAxis="1,0,0"))
   > Set thigh_rot_r(thigh_rot_r=GetAlphaFromReferenceTransformDelta(SkeletalMeshComponent=@n3, BoneName=thigh_r, BaseName=pelvis, InputRange=(X=-20,Y=-100), OutputRange=(X=0,Y=0.75), TwistAxis="1,0,0"))
   n4 = Get Skel_Mesh()
   > Set calf_rot_l(calf_rot_l:double=GetAlphaFromReferenceTransformDelta(SkeletalMeshComponent=@n4, BoneName=calf_l, BaseName=thigh_l, InputRange=(X=20,Y=100), TwistAxis="1,0,0")) -> (Output_Get:double)
   > Set calf_rot_r(calf_rot_r:double=GetAlphaFromReferenceTransformDelta(SkeletalMeshComponent=@n4, BoneName=calf_r, BaseName=thigh_r, InputRange=(X=20,Y=100), TwistAxis="1,0,0")) -> (Output_Get:double)
   n5 = Get Skel_Mesh()
   Set thigh_rot_l(
    execute=@n2.else
    thigh_rot_l=GetAlphaFromReferenceTransformDelta(SkeletalMeshComponent=@n5, BoneName=thigh_l, BaseName=pelvis, InputRange=(X=-20,Y=-100), OutputRange=(X=0.3,Y=0.75), TwistAxis="1,0,0")
   )
   > Set thigh_rot_r(thigh_rot_r=GetAlphaFromReferenceTransformDelta(SkeletalMeshComponent=@n5, BoneName=thigh_r, BaseName=pelvis, InputRange=(X=-20,Y=-100), OutputRange=(X=0.3,Y=0.75), TwistAxis="1,0,0"))
   n6 = Get Skel_Mesh()
   > Set calf_rot_l(calf_rot_l=GetAlphaFromReferenceTransformDelta(SkeletalMeshComponent=@n6, BoneName=calf_l, BaseName=thigh_l, InputRange=(X=20,Y=100), OutputRange=(X=0.3,Y=0.7), TwistAxis="1,0,0"))
   > Set calf_rot_r(calf_rot_r=GetAlphaFromReferenceTransformDelta(SkeletalMeshComponent=@n6, BoneName=calf_r, BaseName=thigh_r, InputRange=(X=20,Y=100), OutputRange=(X=0.3,Y=0.7), TwistAxis="1,0,0"))
   n7 = ComplexSimSpaceFunction DEPRECATED!(
    execute=@n1.then_1
    CharacterPartAnimInstance:Custom Character Part Anim Instance Object Reference=Self-Reference()
    InterpSpeedMax:double=9
    PawnSpeedMax:double=20
    RInterpErrorTolerance:double=1
    DeltaTime:double=0
    DebugDraw:bool=false
    sim_space_rotation_in:Rotator=Get sim_space_rotation() -> (sim_space_rotation:Rotator)
    inComponentSpaceIn:bool=Get InComponentSpace() -> (InComponentSpace:bool)
    inComponentSpaceLastFrameIn:bool=Get InComponentSpaceLastFrame() -> (InComponentSpaceLastFrame:bool)
    lockedOnSimSpaceIn:bool=Get LockedOnSimSpace() -> (LockedOnSimSpace:bool)
    InterpSpeedIn:double=Get InterpSpeed() -> (InterpSpeed:double)
   ) -> (
    then
    sim_space_rotation:Rotator
    inComponentSpaceOut:bool
    inComponentSpaceLastFrameOut:bool
    lockedOnSimSpaceOut:bool
    InterpSpeedOut:double
   )
   > Set sim_space_rotation(sim_space_rotation:Rotator=@n7) -> (Output_Get:Rotator)
   > Set InComponentSpace(InComponentSpace:bool=@n7.inComponentSpaceOut) -> (Output_Get:bool)
   > Set InComponentSpaceLastFrame(InComponentSpaceLastFrame:bool=@n7.inComponentSpaceLastFrameOut) -> (Output_Get:bool)
   > Set LockedOnSimSpace(LockedOnSimSpace:bool=@n7.lockedOnSimSpaceOut) -> (Output_Get:bool)
   > Set InterpSpeed(InterpSpeed:double=@n7.InterpSpeedOut) -> (Output_Get:double)
   n8 = Gravity Override State Function DEPRECATED(
    execute=@n1.then_2
    CharacterPartAnimInstance=Self-Reference()
    GameTimeInSeconds=GetGameTimeInSeconds()
    GravityOverrideParams=Get Gravity Override Params()
   )
   > Set GravityOverride(GravityOverride=@n8)
   n9 = Get Skel_Mesh()
   Set Upperarm l(
    execute=@n1.then_4
    Upperarm l:double=GetAlphaFromReferenceTransformDelta(SkeletalMeshComponent=@n9, BoneName=upperarm_l, BaseName=spine_05, InputRange=(X=50,Y=170), TwistAxis="1,0,0")
   ) -> (Output_Get:double)
   > Set Upperarm r(Upperarm r:double=GetAlphaFromReferenceTransformDelta(SkeletalMeshComponent=@n9, BoneName=clavicle_r, BaseName=spine_05, InputRange=(X=30,Y=125), TwistAxis="1,0,0")) -> (Output_Get:double)
   n10 = Sequence!(execute=@n1.then_5) -> (then_0, then_1, then_2, then_3, then_4)
   n11 = GetAnimInstance(self=GetOwnerSkeletalMeshComponent())
   > Is Valid(InputObject=@n11)
   > n12 = Set Istractior(Istractior:bool=Montage_IsPlaying(self=@n11, Montage=/Game/Animation/Game/MainPlayer/Emotes/Traction/Emote_Traction_CMM_M)) -> (Output_Get:bool)
   > n13 = Branch(Condition=@n12)
   > Set IstractiorEmote(IstractiorEmote:double=1) -> (Output_Get:double)
   Set IstractiorEmote(execute=@n13.else, IstractiorEmote=0)
   Is Valid(exec=@n10.then_1, InputObject=@n11)
   > n14 = Set Issunburstheart(Issunburstheart:bool=Montage_IsPlaying(self=@n11, Montage=/BRCosmetics/Animation/Game/MainPlayer/Emotes/SunBurstHeart/CMM/Emote_SunBurstHeart_CMM_M)) -> (Output_Get:bool)
   > n15 = Branch(Condition=@n14)
   > Set sunburstheartEmote(sunburstheartEmote:double=1) -> (Output_Get:double)
   Set sunburstheartEmote(execute=@n15.else, sunburstheartEmote=0)
   Is Valid(exec=@n10.then_2, InputObject=@n11)
   > n16 = Set isAlfredo(isAlfredo:bool=Montage_IsPlaying(self=@n11, Montage=/BRCosmetics/Animation/Game/MainPlayer/Emotes/Alfredo/Emote_Alfredo_CMM_M)) -> (Output_Get:bool)
   > n17 = Branch(Condition=@n16)
   > Set IsAlfredoEmote(IsAlfredoEmote:double=1) -> (Output_Get:double)
   Set IsAlfredoEmote(execute=@n17.else, IsAlfredoEmote=0)
   Is Valid(exec=@n10.then_3, InputObject=@n11)
   > n18 = Set isCustomEmotePlaying(isCustomEmotePlaying:bool=OR Boolean!(
     A=Montage_IsPlaying(self=@n11, Montage=&Emote_SunBurstHeart_CMM_M)
     B=Montage_IsPlaying(self=@n11, Montage=&Emote_Alfredo_CMM_M)
     C:bool=Montage_IsPlaying(self=@n11, Montage=&Emote_Traction_CMM_M)
     D:bool=Montage_IsPlaying(self=@n11, Montage=/BRCosmetics/Animation/Game/MainPlayer/Emotes/Alfredo/Emote_Alfredo_CMM_M_Frontend)
     E:bool=Montage_IsPlaying(self=@n11, Montage=/Game/Animation/Game/MainPlayer/Emotes/Traction/Emote_Traction_CMM_FrontEnd_M)
     F:bool
    ) -> (ReturnValue)) -> (Output_Get:bool)
   > n19 = Branch(Condition=@n18)
   > Set skirt rban(skirt rban:double=0) -> (Output_Get:double)
   Set skirt rban(execute=@n19.else, skirt rban=0.75)
   n20 = Sequence!(execute=@n10.then_4) -> (then_0, then_1, then_2)
   n21 = GetAnimInstance(self=GetOwnerSkeletalMeshComponent())
   > Is Valid(InputObject=@n21)
   > n22 = Set Istractior(Istractior=Montage_IsPlaying(self=@n21, Montage=&Emote_Traction_CMM_FrontEnd_M))
   > n23 = Branch(Condition=@n22)
   > Set IstractiorEmoteFrontEnd(IstractiorEmoteFrontEnd:double=1) -> (Output_Get:double)
   Set IstractiorEmoteFrontEnd(execute=@n23.else, IstractiorEmoteFrontEnd=0)
   Is Valid(exec=@n20.then_2, InputObject=@n21)
   > n24 = Set isAlfredo(isAlfredo=Montage_IsPlaying(self=@n21, Montage=&Emote_Alfredo_CMM_M_Frontend))
   > n25 = Branch(Condition=@n24)
   > Set IsAlfredoEmoteFrontEnd(IsAlfredoEmoteFrontEnd:double=1) -> (Output_Get:double)
   Set IsAlfredoEmoteFrontEnd(execute=@n25.else, IsAlfredoEmoteFrontEnd=0)
   n26 = Event OnWeaponChanged()
   n27 = Get Dynamic Material Instance(execute=[@n26.then, @n35.then], SkeletalMeshComponent=GetPartSkeletalMeshComponent())
   > n28 = Sequence!(execute) -> (then_0, then_1)
   n29 = PickaxeItemDefinitionHasAnyMetaTag(Tags=Make Array("[0]"=(TagName="Cosmetics.MetaTags.KittyWarrior")), bRequiresEquipped=true)
   > n30 = Branch(Condition=@n29)
   > n31 = Branch(Condition=Get bIsSkyDiving())
   SetScalarParameterValue(execute=[@n33.then, @n32.else, @n31.then, @n30.else], self=@n27, ParameterName=Hide Element 10, Value=0)
   SetScalarParameterValue(execute=[@n32.then, @n31.else], self=@n27, ParameterName=Hide Element 10, Value=1)
   n32 = Branch(execute=[@n34.then_0, @n28.then_1], Condition=@n29)
   n33 = Event OnBeginSkydiving()
   Event BlueprintInitializeAnimation()
   > Set Skel_Mesh(Skel_Mesh=GetOwnerSkeletalMeshComponent())
   Event OnEndSkydiving()
   > n34 = Sequence!(execute) -> (then_0, then_1, then_2:Exec)
   Delay(execute=@n34.then_1, Duration=0.1)
   > ResetDynamics(InTeleportType=TeleportPhysics)
   > ResetDynamics(InTeleportType=ResetPhysics)
   n35 = Event OnCharacterCustomizationLoaded()
   n36 = Branch(Condition=Get bIsMoving2D())
   > Set skirt rban(skirt rban=0.75)
   Set skirt rban(execute=@n36.else, skirt rban=1)
   Set Skel_Mesh(Skel_Mesh=@n35.CustomBaseMesh)
   > Branch(Condition=Get bIsInFrontEnd())
  ]}
]
variables [
 AlphaSittingEmote {type=number}
 calf_rot_l {type=number}
 calf_rot_r {type=number}
 Gravity Override Params {
  type=GravityOverrideParamsStruct
  default {emoteOrMelee {gravityOverride=<0 0 -720>}, onGround_Moving {gravityOverride=<0 0 -720>}, onGround_Standing {gravityOverride=<0 0 -720>}}
 }
 GravityOverride {type=Vector}
 InComponentSpace {type=boolean}
 InComponentSpaceLastFrame {type=boolean}
 InterpSpeed {type=number}
 isAlfredo {type=boolean}
 IsAlfredoEmote {type=number}
 IsAlfredoEmoteFrontEnd {type=number}
 isCustomEmotePlaying {type=boolean}
 Issunburstheart {type=boolean}
 Istractior {type=boolean}
 IstractiorEmote {type=number}
 IstractiorEmoteFrontEnd {type=number}
 LockedOnSimSpace {type=boolean}
 sim_space_rotation {type=Rotator}
 Skel_Mesh {type=SkeletalMeshComponent}
 skirt rban {type=number, default=0.75}
 sunburstheartEmote {type=number}
 sunburstheartEmoteFrontEnd {type=number}
 thigh_rot_l {type=number}
 thigh_rot_r {type=number}
 Upperarm l {type=number}
 Upperarm r {type=number}
]
