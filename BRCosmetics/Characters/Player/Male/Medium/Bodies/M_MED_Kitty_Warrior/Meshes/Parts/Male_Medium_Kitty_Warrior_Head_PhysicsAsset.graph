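# Physics asset for the head part of the male medium Kitty Warrior player cosmetic, setting up physics for its two dynamic ear bones.
# Each ear bone has a capsule body, and the head bone has a very small capsule that serves as their shared attachment.
# One constraint per ear links it to the head with limited swing and a much tighter twist limit.
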PhysicsAsset /BRCosmetics/Characters/Player/Male/Medium/Bodies/M_MED_Kitty_Warrior/Meshes/Parts/Male_Medium_Kitty_Warrior_Head_PhysicsAsset
bodies {
 dyn_ear_l [{shapeType=Capsule, length=4.9815, radius=4.0429, shapeName=dyn_ear_l_capsule, center=<5 0 0>, rotation=<rot 90 0 0>}]
 dyn_ear_r [{shapeType=Capsule, length=4.9998, radius=3.9746, shapeName=dyn_ear_r_capsule, center=<5 0 0>, rotation=<rot 90 0 0>}]
 head [{shapeType=Capsule, length=0.505, radius=0.505, center=<8.0423 -0.4979 0.1465>, rotation=<rot 0 -124.5948 -89.4097>}]
}
constraints [
 {bone1Name=dyn_ear_r, bone2Name=head, swing1LimitDegrees=30, swing2LimitDegrees=30, twistLimitDegrees=5}
 {bone1Name=dyn_ear_l, bone2Name=head, swing1LimitDegrees=30, swing2LimitDegrees=30, twistLimitDegrees=5}
]
